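# Translucent UI material for the horizontal backplate of seasonal goals progress, judging by its name.
# A CH7 bonus-goals quest background function fades toward zero through a smoothstepped vertical UV gradient, tuned by offset and smoothness.
# The result is multiplied by a rounded-corner shape mask with per-corner radii.
# IsCut sets top radii to 1 and bottom radii to the BottomRight value, while IsSquaredOff sets every radius to 1.
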Material /FortniteCoreUI/UI/Quests/Materials/M_UI_BP_SeasonalGoalsProgressHorizontalBack
graph {
 properties {
  bAutomaticallySetUsageInEditor=false
  bUsedWithStaticMesh=false
  blendMode=BLEND_Translucent
  materialDomain=MD_UI
  usedShadingModels=Default Lit
 }
 nodes [
  n1 = TextureCoordinate()
  n2 = ScalarParameter() {defaultValue=0.1, parameterName=GradientOffset}
  n3 = Multiply(A=ComponentMask(None=Add(A=@n1, B=@n2) {constB=0.2}) {g=true}, B=ScalarParameter() {defaultValue=0.6, parameterName=Smoothness}) {constB=0.5}
  n4 = VectorParameter() {
   group=Shape
   parameterName=CornerRadius
   sortPriority=1
   channelNames {a=BottomRight, b=TopRight, g=BottomLeft, r=TopLeft}
   defaultValue=<rgba 16 16 16 16>
  }
  n5 = Constant() {r=1}
  n6 = LinearInterpolate(
   A=AppendVector(A=@n4, B=@n4)
   B=MaterialFunctionCall(R=@n5, G=@n4.BottomRight, B=@n5, A=@n4.BottomRight) {materialFunction=&AppendMany}.RGBA
   Alpha=ScalarParameter() {desc=Whether or not the backplate is chopped, parameterName=IsCut}
  )
  n7 = Constant() {r=1}
  n8 = ScalarParameter() {desc=All corner radiuses will be replaced with 1 if this is true, parameterName=IsSquaredOff}
  Multiply(
   A=LinearInterpolate(
    A=MaterialFunctionCall() {materialFunction=/FortniteCoreUI/UI/Quests/Materials/CH7_materials/MF_UI_CH7_BackgroundQuestsBonusGoals}.Result
    Alpha=SmoothStep(Value=@n3)
   ) {constA=1, constAlpha=1, constB=0}
   B=MaterialFunctionCall(
    UV=@n1
    Corner Radius=LinearInterpolate(A=@n6, B=@n7, Alpha=@n8)
    Radius in Pixels=ScalarParameter() {defaultValue=1, group=Shape, parameterName=Radius in Pixels, sliderMax=1, sortPriority=2}
   ) {materialFunction=&MF_UI_CoreUI_RoundedMask}.Shape Mask
  )
 ]
}
