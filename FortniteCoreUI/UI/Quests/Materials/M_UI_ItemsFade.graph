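# Translucent UI material in the quest UI that, judging by its name, fades items.
# Texture coordinates pass through CustomRotator, angled by a Rotation scalar parameter, into a linear gradient function.
# A GetSlatePost0 function call sits alongside to read the Slate post buffer.
# The graph shows no link between that call and the gradient, nor any connection to the material outputs.
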Material /FortniteCoreUI/UI/Quests/Materials/M_UI_ItemsFade
graph {
 properties {
  bUsedWithStaticMesh=false
  blendMode=BLEND_Translucent
  materialDomain=MD_UI
  pixelDepthOffsetMode=PDOM_Legacy
  usedShadingModels=Default Lit
 }
 nodes [
  MaterialFunctionCall!("UVs (Optional)") {materialFunction=/Engine/Functions/UserInterface/GetSlatePost0}
  MaterialFunctionCall(UVs=MaterialFunctionCall(UVs=TextureCoordinate(), "Rotation Angle (0-1)"=ScalarParameter() {parameterName=Rotation}) {materialFunction=&CustomRotator}.Rotated Values) {materialFunction=&MF_UI_LinearGradient}
 ]
}
