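# Translucent UI material for a linear progress bar, filed under quest UI, with separately tinted empty and filled areas.
# Progress interpolates from ProgressStart to ProgressEnd via AnimateProgress, optionally snapping to segment steps.
# That value drives either a SmoothStep edge or a rounded MF_UI_Box fill, giving the mask that blends color and opacity.
# An optional pixel-sized outline is converted to normalized units to inset and mask the fill.
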Material /FortniteCoreUI/UI/Quests/Materials/M_UI_LinearProgressBar
graph {
 properties {
  bUsedWithStaticMesh=false
  blendMode=BLEND_Translucent
  materialDomain=MD_UI
  pixelDepthOffsetMode=PDOM_Legacy
  usedShadingModels=Default Lit
 }
 nodes [
  n1 = VectorParameter() {
   desc="Color for empty part of progress bar.\r\nAlpha defines opacity for empty part of progress bar."
   group=Color
   parameterName=EmptyColor
   sortPriority=10
   defaultValue=<rgba 0.3924 0.3924 0.3924 0.2>
  }
  n2 = VectorParameter() {desc=Start color in gradient, group=Color, parameterName=ColorA, sortPriority=1, defaultValue=<rgba 0 0 1 1>}
  LinearInterpolate(A=@n1.RGB, B=@n2.RGB, Alpha=Max(B=NamedRerouteUsage()))
  n3 = VectorParameter() {
   desc="Color for empty part of progress bar.\r\nAlpha defines opacity for empty part of progress bar."
   group=Color
   parameterName=EmptyColor
   sortPriority=10
   defaultValue=<rgba 0.3924 0.3924 0.3924 0.2>
  }
  n4 = ScalarParameter() {defaultValue=1, desc="Alpha for filled part of progress bar.\r\n0 -1", group=Color, parameterName=FillAlpha, sliderMax=1}
  Multiply(A=LinearInterpolate(A=@n3.A, B=@n4, Alpha=NamedRerouteUsage()), B=NamedRerouteUsage())
  n5 = ScalarParameter() {group=Progress_Layout, parameterName=ProgressEdgeSharpness, sliderMax=1, sortPriority=2}
  n6 = ScalarParameter() {
   desc="When progress value changes, it defines where change starts.\r\n0 - 1"
   group=Progress
   parameterName=ProgressStart
   sliderMax=1
   sortPriority=1
  }
  n7 = ScalarParameter() {
   defaultValue=1
   desc="When progress value changes, it defines where change ends.\r\n0 - 1"
   group=Progress
   parameterName=ProgressEnd
   sliderMax=1
   sortPriority=2
  }
  n8 = ScalarParameter() {
   defaultValue=1
   desc="Property for animating progress value change (from ProgressStart to ProgressEnd)\r\n0 - 1"
   group=Progress
   parameterName=AnimateProgress
   sliderMax=1
   sortPriority=3
  }
  n9 = LinearInterpolate(A=@n6, B=@n7, Alpha=@n8)
  n10 = ScalarParameter() {defaultValue=8, desc=Number of segments in progress bar, group=Segments, parameterName=SegmentsCount, sortPriority=2}
  n11 = LinearInterpolate(
   A=@n9
   B=MaterialFunctionCall!("Value (0-1)"=@n9, Steps=Floor(None=@n10)) {materialFunction=/FortUILibrary/MaterialFunctions/Basics/MF_UI_StepRatio}.Stepped Ratio
   Alpha=ScalarParameter() {desc="Change progress value only with segments chanks.\r\nbool", group=Segments, parameterName=SnapToSegments, sliderMax=1}
  )
  n12 = ComponentMask(None=Add(A=NamedRerouteUsage(), B=NamedRerouteDeclaration(None=@n11) {name=Progress, nodeColor=<rgba 0 0 0 1>}) {constB=0.5}) {r=true}
  n13 = ScalarParameter() {
   defaultValue=1
   desc="Will use rounded box to show progress value. Not support snap to segments.\r\nbool"
   group=Progress_Layout
   parameterName=UseRoundedProgress
   sliderMax=1
  }
  n14 = LinearInterpolate(A=Saturate(None=SmoothStep(Max=@n5, Value=@n12) {constMax=0.01}), B=NamedRerouteUsage(), Alpha=@n13)
  n15 = NamedRerouteUsage()
  n16 = ScalarParameter() {
   defaultValue=1
   desc="Will add outline for progress bar.\r\nDoesn't work correct with \"snap to segments\".\r\nbool"
   group=Outline
   parameterName=UseOutline
   sliderMax=1
   sortPriority=1
  }
  NamedRerouteDeclaration(None=LinearInterpolate(A=@n14, B=Multiply(A=@n14, B=@n15), Alpha=@n16)) {name=ProgressMask, nodeColor=<rgba 0 0 0 1>}
  n17 = TextureCoordinate()
  NamedRerouteDeclaration(None=@n17) {name=UVs, nodeColor=<rgba 0 0 0 1>}
  NamedRerouteDeclaration(None=@n14) {name=ProgressMaskForDelta, nodeColor=<rgba 0 0 0 1>}
  n18 = Add(A=Divide(A=Add(A=@n17) {constB=-0.5}, B=NamedRerouteUsage())) {constB=0.5}
  NamedRerouteDeclaration(None=Multiply(A=@n18) {constB=-1}) {name=RemappedUVs, nodeColor=<rgba 0 0 0 1>}
  NamedRerouteDeclaration(None=@n18) {name=UVsWithoutOutline, nodeColor=<rgba 0 0 0 1>}
  n19 = ScalarParameter() {
   defaultValue=1
   desc="Will add outline for progress bar.\r\nDoesn't work correct with \"snap to segments\".\r\nbool"
   group=Outline
   parameterName=UseOutline
   sliderMax=1
   sortPriority=1
  }
  n20 = LinearInterpolate(
   B=ScalarParameter() {defaultValue=10, desc=Size of outline for rounded progress bar in px., group=Outline, parameterName=Outline_PX}
   Alpha=@n19
  )
  n21 = Saturate(None=Multiply(A=MaterialFunctionCall!(Value PX=@n20, UV Pixel Size, X or Y) {materialFunction=/FortUILibrary/MaterialFunctions/Basics/MF_UI_PixelsToNormalized}.Normalized) {constB=2})
  NamedRerouteDeclaration(None=@n21) {name=OutlineHorNormalized, nodeColor=<rgba 0 0 0 1>}
  n22 = NamedRerouteUsage()
  n23 = OneMinus(None=NamedRerouteUsage())
  n24 = NamedRerouteUsage()
  n25 = AppendVector(A=Add(A=ComponentMask(None=@n22) {r=true}, B=Multiply(A=@n23, B=@n24)), B=ComponentMask(None=@n22) {g=true})
  n26 = Step(X=TextureCoordinate()) {constY=0.5}
  n27 = LinearInterpolate(
   A=MaterialFunctionCall!(Texture Scale=Constant2Vector() {g=1, r=2}, UVs=@n25, Offset XY Texturesize Z=Constant3Vector() {constant=<rgba 1 0.5 1 0>}) {materialFunction=/Engine/Functions/Engine_MaterialFunctions01/Texturing/ScaleUVsAroundPoint}.UVs
   B=@n25
   Alpha=@n26
  )
  n28 = NamedRerouteUsage()
  n29 = AppendVector(A=@n24, B=@n28)
  n30 = MaterialFunctionCall(UVs=@n25, Size=@n29, CornerRoundness=@n28, KeepAspectRatio=StaticBool() {value=true}) {materialFunction=&MF_UI_Box}
  n31 = NamedRerouteUsage()
  NamedRerouteDeclaration(None=LinearInterpolate(
    A=MaterialFunctionCall(UVs=@n27, Size=@n29) {materialFunction=&MF_UI_Box}.Fill
    B=Multiply(A=@n30.Fill, B=@n31)
    Alpha=@n26
   )) {name=RoundedProgressMask, nodeColor=<rgba 0 0 0 1>}
  n32 = Constant2Vector() {g=1, r=1.1}
  n33 = TextureCoordinate()
  n34 = Constant3Vector() {constant=<rgba 1 0.5 1 0>}
  n35 = ComponentMask(None=@n33) {r=true}
  n36 = MaterialFunctionCall(
   UVs=LinearInterpolate(
    A=MaterialFunctionCall(Texture Scale=@n32, UVs=@n33, Offset XY Texturesize Z=@n34) {materialFunction=&ScaleUVsAroundPoint}.UVs
    B=@n33
    Alpha=Step(X=@n35) {constY=0.2}
   )
   Size=Constant2Vector() {g=1, r=1}
   CornerRoundness=Constant4Vector() {constant=<rgba 0 0 1 1>}.RGBA
   OutlineThickness=NamedRerouteUsage()
   KeepAspectRatio=StaticBool() {value=true}
  ) {materialFunction=&MF_UI_Box}
  NamedRerouteDeclaration(None=@n36.Fill) {name=RoundedMask, nodeColor=<rgba 0 0 0 1>}
  n37 = Constant() {r=1}
  n38 = Saturate(None=Multiply(A=MaterialFunctionCall(Value PX=@n20, X or Y=@n37) {materialFunction=&MF_UI_PixelsToNormalized}.Normalized) {constB=2})
  NamedRerouteDeclaration(None=OneMinus(None=@n38)) {name=BarWithoutOutlineVertNormalized, nodeColor=<rgba 0 0 0 1>}
  NamedRerouteDeclaration(None=OneMinus(None=@n21)) {name=BarWithoutOutlineHorNormalized, nodeColor=<rgba 0 0 0 1>}
  NamedRerouteDeclaration(None=@n38) {name=OutlineVertNormalized, nodeColor=<rgba 0 0 0 1>}
  NamedRerouteDeclaration(None=OneMinus(None=@n36.Outline)) {name=OutlineMask, nodeColor=<rgba 0 0 0 1>}
 ]
}
