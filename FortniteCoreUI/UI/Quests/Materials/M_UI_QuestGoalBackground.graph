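# Translucent UI material for quest goal backgrounds, building a shape from a rounded box, masked triangle and radial step mask.
# Modular button-state functions, with an alternate HMX color layout, color the shape, which can screen-blend with contrast-adjusted GetSlatePost0.
# An avatar texture supports scale animation and channel packing, with an optional outline above or below it.
# Named reroutes link these parts, alongside vignette and animated archived-gradient alphas.
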Material /FortniteCoreUI/UI/Quests/Materials/M_UI_QuestGoalBackground
graph {
 properties {
  bAutomaticallySetUsageInEditor=false
  bUsedWithStaticMesh=false
  blendMode=BLEND_Translucent
  materialDomain=MD_UI
  usedShadingModels=Default Lit
 }
 nodes [
  n1 = NamedRerouteUsage()
  n2 = NamedRerouteUsage()
  n3 = OneMinus(None=Saturate(None=Subtract(A=@n1, B=@n2)))
  n4 = TextureCoordinate()
  Add(A=@n3, B=Step(Y=ComponentMask(None=@n4) {g=true}) {constX=0.7, constY=0.7})
  n5 = NamedRerouteUsage()
  n6 = VectorParameter() {group=Style, parameterName=OutlineColor, defaultValue=<rgba 0.9301 1 0.0103 1>}
  n7 = LinearInterpolate(A=@n5, B=@n6.RGB, Alpha=NamedRerouteUsage())
  n8 = Min(
   A=OneMinus(None=ScalarParameter() {group=Style, parameterName=ShowOutlineAboveAvatar, sliderMax=1})
   B=ScalarParameter() {group=Style, parameterName=UseOutline}
  )
  n9 = NamedRerouteUsage()
  n10 = Multiply(A=Multiply(A=OneMinus(None=@n9), B=NamedRerouteUsage()), B=NamedRerouteUsage())
  n11 = ScalarParameter() {parameterName=UseAvatar}
  n12 = ScalarParameter() {group=Style, parameterName=ShowOutlineAboveAvatar, sliderMax=1}
  n13 = ScalarParameter() {group=Style, parameterName=UseOutline}
  LinearInterpolate(
   A=LinearInterpolate(A=LinearInterpolate(A=@n5, B=@n7, Alpha=@n8), B=NamedRerouteUsage(), Alpha=Min(A=@n10, B=@n11))
   B=@n6.RGB
   Alpha=Min(A=Min(A=@n12, B=@n13), B=NamedRerouteUsage())
  ) {constAlpha=1}
  n14 = Multiply(A=NamedRerouteUsage(), B=NamedRerouteUsage())
  n15 = LinearInterpolate(A=@n14, B=LinearInterpolate(A=@n14, Alpha=NamedRerouteUsage()), Alpha=ScalarParameter() {group=Style, parameterName=UseOutline})
  n16 = NamedRerouteUsage()
  n17 = Multiply(A=Multiply(A=OneMinus(None=@n16), B=NamedRerouteUsage()), B=NamedRerouteUsage())
  n18 = NamedRerouteUsage()
  LinearInterpolate(A=@n15, B=LinearInterpolate(A=@n15, Alpha=Multiply(A=@n17, B=@n18)), Alpha=ScalarParameter() {parameterName=UseAvatar})
  n19 = MaterialFunctionCall(Float2=TextureCoordinate()) {materialFunction=&BreakOutFloat2Components}
  n20 = ScalarParameter() {parameterName=BoxOffset_X}
  n21 = ScalarParameter() {defaultValue=-0.05, parameterName=BoxOffset_Y}
  n22 = VectorParameter() {group=Corners, parameterName=Corner Radius, sortPriority=2, defaultValue=<rgba 0.33 0.33 0.33 0.33>}
  n23 = MaterialFunctionCall(
   UVs=MaterialFunctionCall(X=Subtract(A=@n19.R, B=@n20) {constB=0}, Y=Subtract(A=@n19.G, B=@n21) {constB=-0.05}) {materialFunction=&MakeFloat2}.Result
   Size=VectorParameter() {parameterName=BoxSize, defaultValue=<rgba 1 0.9 0 1>}.RGB
   CornerRoundness=AppendVector(A=@n22.RGB, B=@n22.RGB)
   OutlineThickness=ScalarParameter() {defaultValue=0.036, parameterName=OutlineThickness}
   OutlineAlignment=Constant() {r=-1}
   KeepAspectRatio=StaticBool() {value=true}
  ) {materialFunction=&MF_UI_Box}
  NamedRerouteDeclaration(None=@n23.Outline) {name=BoxOutline, nodeColor=<rgba 0.6172 1 0 1>}
  n24 = LinearInterpolate(
   B=LinearInterpolate(B=NamedRerouteUsage(), Alpha=ScalarParameter() {defaultValue=1, parameterName=UseMask}) {constA=1}
   Alpha=ScalarParameter() {defaultValue=1, parameterName=UseTriangle}
  )
  n25 = MaterialFunctionCall!(
   UVs=TextureCoordinate()
   Corner 01=VectorParameter() {parameterName=Corner_0, defaultValue=<rgba -1.3 -0.1 0 1>}.RGB
   Corner 02=VectorParameter() {parameterName=Corner_1, defaultValue=<rgba 0 1 0 1>}.RGB
   Corner 03=VectorParameter() {parameterName=Corner_2, defaultValue=<rgba 1.3 -0.1 0 1>}.RGB
   Glow Min
   Glow Max
   Outline Thickness=ScalarParameter() {defaultValue=0.036, parameterName=OutlineThickness}
  ) {materialFunction=/FortUILibrary/MaterialFunctions/Shapes/MF_UI_WIP_Triangle}
  NamedRerouteDeclaration(None=Multiply(
    A=@n24
    B=MaterialFunctionCall(DistanceField=@n25.SDF Outline) {materialFunction=&MF_UI_AntiAliasing}.Inner
   )) {name=TriangleOutline, nodeColor=<rgba 1 0 0.9375 1>}
  n26 = AppendVector(
   A=ScalarParameter() {defaultValue=0.5, group=Vignette, parameterName=VignetteCenter_X, sortPriority=10}
   B=ScalarParameter() {group=Vignette, parameterName=VignetteCenter_Y, sortPriority=11}
  )
  n27 = ScalarParameter() {defaultValue=1.4, group=Vignette, parameterName=VignetteRadius, sortPriority=12}
  n28 = ScalarParameter() {defaultValue=3.5, group=Vignette, parameterName=VignetteDensity, sortPriority=13}
  n29 = Multiply(
   A=OneMinus(None=MaterialFunctionCall(CenterPosition=@n26, Radius=@n27, Density=@n28) {materialFunction=&RadialGradientExponential}.RadialGradientExponential)
   B=ScalarParameter() {defaultValue=1, group=Vignette, parameterName=VignetteStrength, sortPriority=6}
  )
  n30 = ScalarParameter() {
   defaultValue=1
   desc="(0 or 1) If enabled applies vignette effect on the item texture."
   group=Vignette
   parameterName=UseVignette
   sliderMax=1
   sortPriority=0
  }
  NamedRerouteDeclaration(None=Multiply(A=Saturate(None=@n29), B=@n30)) {name=Vignette Alpha, nodeColor=<rgba 0 0 0 1>}
  n31 = OneMinus(None=MaterialFunctionCall(
    GradientStart=ScalarParameter() {group=Archived, parameterName=ArchivedGradientStart, sortPriority=10}
    GradientEnd=ScalarParameter() {defaultValue=2, group=Archived, parameterName=ArchivedGradientEnd, sortPriority=11}
    RotationAngle=ScalarParameter() {defaultValue=0.75, group=Archived, parameterName=ArchivedGradientAngle, sortPriority=12}
   ) {materialFunction=&MF_UI_LinearGradient}.Result)
  n32 = ScalarParameter() {
   defaultValue=0.95
   desc=Multiplies Archived Colors alpha. A quick way to set desired threshold in applying the archived effect.
   group=Archived
   parameterName=ArchivedAlphaMultiplier
   sliderMax=1
   sortPriority=3
  }
  n33 = ScalarParameter() {
   desc="(0 - 1) Animate item Archiving effect.\r\n0: Item not archived.\r\n1: Item archived."
   group=Archived
   parameterName=ArchivedAnimation
   sliderMax=1
   sortPriority=0
  }
  NamedRerouteDeclaration(None=LinearInterpolate(B=Multiply(A=@n31, B=@n32), Alpha=@n33) {constA=1}) {name=Archived Alpha, nodeColor=<rgba 0 0 0 1>}
  n34 = Multiply(A=NamedRerouteUsage(), B=NamedRerouteUsage())
  n35 = NamedRerouteUsage()
  NamedRerouteDeclaration(None=LinearInterpolate(
    A=@n34
    B=MaterialFunctionCall!(Blend=@n35, Base=@n34) {materialFunction=/Engine/Functions/Engine_MaterialFunctions03/Blends/Blend_Screen}.Result
    Alpha=ScalarParameter() {group=Style, parameterName=UseBlendScreenEffect}
   )) {name=ColoredShape, nodeColor=<rgba 0 0.8125 1 1>}
  n36 = Max(A=Saturate(None=MaterialFunctionCall!("UVs (Optional)") {materialFunction=/Engine/Functions/UserInterface/GetSlatePost0}.LinearRGB)) {constB=0.0001}
  n37 = ScalarParameter() {defaultValue=0.4, group=Style, parameterName=EffectWeakness}
  NamedRerouteDeclaration(None=Saturate(None=MaterialFunctionCall!(Color=@n36, Contrast=@n37) {materialFunction=/Engine/Functions/Engine_MaterialFunctions01/ImageAdjustment/Contrast_Preserve_Color}.Result)) {name=BlendBackground, nodeColor=<rgba 1 0.8672 0 1>}
  n38 = MaterialFunctionCall!(Idle, Idle Opacity, Hovered, Hovered Opacity, Focused, Focused Opacity, Pressed, Pressed Opacity, Disabled, Disabled Opacity) {materialFunction=/FortUILibrary/Systems/ModularButton/MaterialFunctions/MF_UIKit_ButtonStatesBase_Lerp}
  n39 = MaterialFunctionCall!(
   Base States=@n38.Color
   Base States Opacity=@n38.Color
   Selected_Idle
   Selected_Idle Opacity
   Selected_Hovered
   Selected_Hovered Opacity
   Selected_Focused
   Selected_Focused Opacity
   Selected_Pressed
   Selected_Pressed Opacity
   Selected_Disabled
   Selected_Disabled Opacity
  ) {materialFunction=/FortUILibrary/Systems/ModularButton/MaterialFunctions/MF_UIKit_ButtonStatesSelectable_Lerp}
  n40 = MaterialFunctionCall(Idle=VectorParameter() {group=Style, parameterName=BackgroundColor_WithOurline, defaultValue=<rgba 1 0.0251 0 1>}.RGB) {materialFunction=&MF_UIKit_ButtonStatesBase_Lerp}
  n41 = MaterialFunctionCall(Base States=@n40.Color, Base States Opacity=@n40.Color) {materialFunction=&MF_UIKit_ButtonStatesSelectable_Lerp}
  n42 = ScalarParameter() {group=Style, parameterName=UseOutline}
  NamedRerouteDeclaration(None=LinearInterpolate(A=LinearInterpolate(A=@n39.Color, B=@n41.Color, Alpha=@n42), B=NamedRerouteUsage(), Alpha=ScalarParameter() {parameterName=UseHMX})) {name=StateColor, nodeColor=<rgba 0.5469 0 1 1>}
  n43 = ScalarParameter() {group=Style, parameterName=UseOutline}
  NamedRerouteDeclaration(None=LinearInterpolate(A=LinearInterpolate(A=@n39.Opacity, B=@n41.Opacity, Alpha=@n43), B=NamedRerouteUsage(), Alpha=ScalarParameter() {parameterName=UseHMX})) {name=StateOpacity, nodeColor=<rgba 0 1 0.2266 1>}
  n44 = MaterialFunctionCall(Float2=TextureCoordinate()) {materialFunction=&BreakOutFloat2Components}
  n45 = ScalarParameter() {parameterName=BoxOffset_X}
  n46 = ScalarParameter() {defaultValue=-0.05, parameterName=BoxOffset_Y}
  NamedRerouteDeclaration(None=MaterialFunctionCall(X=Subtract(A=@n44.R, B=@n45) {constB=0}, Y=Subtract(A=@n44.G, B=@n46) {constB=-0.05}) {materialFunction=&MakeFloat2}.Result) {name=UV_Mask, nodeColor=<rgba 1 0.9141 0 1>}
  NamedRerouteDeclaration(None=VectorParameter() {parameterName=BoxSize, defaultValue=<rgba 1 0.9 0 1>}.RGB) {name=Size in Color, nodeColor=<rgba 0.5938 0 1 1>}
  NamedRerouteDeclaration(None=@n23.Fill) {name=Box, nodeColor=<rgba 0 1 0.7422 1>}
  n47 = TextureCoordinate()
  n48 = LinearInterpolate(
   A=VectorParameter() {group=Style, parameterName=HMX_BackgroundColor, defaultValue=<rgba 0.521 0.521 0.521 0.502>}.RGB
   B=VectorParameter() {group=Style, parameterName=HMX_BackgroundHilightColor, defaultValue=<rgba 1 0.7668 0 1>}.RGB
   Alpha=Step(Y=MaterialFunctionCall(Float2=@n47) {materialFunction=&BreakOutFloat2Components}.G) {constX=0.5}
  )
  n49 = MaterialFunctionCall() {materialFunction=&MF_UIKit_ButtonStatesBase_Lerp}
  n50 = MaterialFunctionCall(Base States=@n49.Color, Base States Opacity=@n49.Color) {materialFunction=&MF_UIKit_ButtonStatesSelectable_Lerp}
  n51 = ScalarParameter() {desc="If enabled reads corner radius values in pixels.\r\nbool", group=Layout, parameterName=UseValueInPixels, sliderMax=1, sortPriority=1}
  n52 = MaterialFunctionCall(
   UV=NamedRerouteUsage()
   Size=Multiply(A=NamedRerouteUsage(), B=ScalarParameter() {defaultValue=0.834, parameterName=HMX BoxScale}) {constB=0.88}
   Corner Radius=Constant() {r=0.13}
   Radius in Pixels=@n51
   Outline Thickness=ScalarParameter() {defaultValue=0.036, parameterName=OutlineThickness}
  ) {materialFunction=&MF_UI_CoreUI_RoundedMask}
  n53 = SmoothStep(Value=Saturate(None=OneMinus(None=@n52.Outline SDF))) {constMax=0.96, constMin=0.95}
  NamedRerouteDeclaration(None=LinearInterpolate(
    A=LinearInterpolate(A=@n48, B=@n50.Color, Alpha=@n52.Shape Mask)
    B=VectorParameter() {group=Style, parameterName=HMX_InnerShadowColor, defaultValue=<rgba 0.213 0.4792 0 1>}.RGB
    Alpha=Min(A=@n53, B=@n52.Shape Mask)
   )) {name=HMX_Colors, nodeColor=<rgba 0.5234 0 1 1>}
  NamedRerouteDeclaration(None=@n50.Opacity) {name=HMX Opacity, nodeColor=<rgba 0.5 1 0 1>}
  NamedRerouteDeclaration(None=Multiply(
    A=MaterialFunctionCall(DistanceField=@n25.SDF) {materialFunction=&MF_UI_AntiAliasing}.Inner
    B=@n24
   )) {name=Triangle, nodeColor=<rgba 1 0.4219 0 1>}
  n54 = NamedRerouteUsage()
  n55 = Subtract(A=NamedRerouteUsage(), B=@n54)
  n56 = Multiply(A=OneMinus(None=Saturate(None=@n55)), B=Max(A=@n54, B=@n2))
  NamedRerouteDeclaration(None=Saturate(None=Multiply(A=@n56, B=@n3))) {name=Outline, nodeColor=<rgba 0.1016 0 1 1>}
  n57 = ComponentMask(None=TextureCoordinate()) {r=true}
  n58 = ComponentMask(None=TextureCoordinate()) {r=true}
  NamedRerouteDeclaration(None=Multiply(A=Step(X=@n57) {constX=0.9, constY=0.1}, B=Step(Y=@n58) {constX=0.9})) {name=TriangleMask, nodeColor=<rgba 0.7109 1 0 1>}
  n59 = MaterialFunctionCall(Float2=MaterialFunctionCall(UVs=TextureCoordinate(), Texture Scale=ScalarParameter() {defaultValue=1, parameterName=RoundScale}) {materialFunction=&ScaleUVsByCenter}.UVs) {materialFunction=&BreakOutFloat2Components}
  n60 = Subtract(A=@n59.R, B=ScalarParameter() {parameterName=RoundStep_X}) {constB=0}
  n61 = Subtract(A=@n59.G, B=ScalarParameter() {parameterName=RoundStep_Y}) {constB=-0.05}
  n62 = Multiply(A=Subtract(A=MaterialFunctionCall(X=@n60, Y=@n61) {materialFunction=&MakeFloat2}.Result) {constB=0.5}) {constB=2}
  NamedRerouteDeclaration(None=Step(Y=Length(None=@n62) {is=class MaterialExpressionLength}, X=ScalarParameter() {defaultValue=1, parameterName=RoundStep}) {constX=0.97}) {name=RoundMask, nodeColor=<rgba 0.9688 0 1 1>}
  n63 = NamedRerouteUsage()
  n64 = NamedRerouteUsage()
  NamedRerouteDeclaration(None=LinearInterpolate(A=NamedRerouteUsage(), Alpha=Multiply(A=@n63, B=@n64))) {name=Shape, nodeColor=<rgba 0.3359 1 0 1>}
  n65 = VectorParameter() {
   desc=Color that applies to channel-packed texture
   group=ChannelPacked
   parameterName=DynamicColor
   sortPriority=2
   defaultValue=<rgba 1 0 1 1>
  }
  n66 = ScalarParameter() {defaultValue=1, desc=Start value for scale animation., group=Item, parameterName=ScaleFrom, sortPriority=2}
  n67 = ScalarParameter() {defaultValue=1.1, desc=End value for scale animation., group=Item, parameterName=ScaleTo, sortPriority=3}
  n68 = ScalarParameter() {desc="(0 - 1) Animate scale transition 'ScaleFrom' - 'ScaleTo'", group=Item, parameterName=ScaleAnimation, sliderMax=1, sortPriority=4}
  n69 = Subtract(
   A=MaterialFunctionCall(
    UVs=MaterialFunctionCall() {materialFunction=&MF_UI_UVScaling}.ScaleY
    Texture Scale=LinearInterpolate(A=@n66, B=@n67, Alpha=@n68)
   ) {materialFunction=&ScaleUVsByCenter}.UVs
   B=AppendVector(A=ScalarParameter() {group=Item, parameterName=ItemOffset_X}, B=ScalarParameter() {group=Item, parameterName=ItemOffset_Y})
  )
  n70 = TextureSampleParameter2D(UVs=@n69) {
   group=Item
   parameterName=Item
   samplerSource=SSM_Clamp_WorldGroupSettings
   sortPriority=1
   texture=/FortUILibrary/Systems/Items/Textures/T_UI_SoldierCommando_F
  }
  n71 = LinearInterpolate(A=LinearInterpolate(B=@n65.RGB, Alpha=@n70.B), Alpha=@n70.G)
  n72 = ScalarParameter() {
   desc="(0 or 1) If enabled switch to work with channel-packed texture."
   group=ChannelPacked
   parameterName=UseChannelPacked
   sliderMax=1
   sortPriority=1
  }
  NamedRerouteDeclaration(None=LinearInterpolate(A=@n70.RGB, B=LinearInterpolate(A=@n71, Alpha=@n70.R) {constB=0}, Alpha=@n72)) {name=Item, nodeColor=<rgba 0 0 0 1>}
  n73 = LinearInterpolate(
   A=MaterialFunctionCall!(UVs=@n69, Alpha=@n70.A) {materialFunction=/FortUILibrary/MaterialFunctions/Basics/MF_UI_ClipEdgePixel}.Result
   Alpha=ScalarParameter() {group=Item, parameterName=RemoveItem}
  ) {constB=0}
  n74 = Saturate(None=Add(A=@n70.B, B=Add(A=@n70.R, B=@n70.G)))
  NamedRerouteDeclaration(None=Multiply(A=LinearInterpolate(A=@n73, B=@n74, Alpha=@n72), B=ScalarParameter() {defaultValue=1, group=Item, parameterName=AvatarOpacity})) {name=Item Alpha, nodeColor=<rgba 0 0 0 1>}
 ]
}
