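# Translucent UI material that draws a quest pin from the quest pin icon texture.
# The icon is sampled with center-scaled, clamped UVs, then run through several SmoothStep thresholds to derive layered masks.
# A procedural circle from MF_UI_Circle, offset vertically, is cut by the inverted icon mask.
# Lerps combine the layers into shading, while a separate multiply masks the icon shape outside that circle.
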Material /FortniteCoreUI/UI/Quests/Materials/M_UI_QuestPin
graph {
 properties {
  bAutomaticallySetUsageInEditor=false
  bUsedWithStaticMesh=false
  blendMode=BLEND_Translucent
  materialDomain=MD_UI
  usedShadingModels=Default Lit
 }
 nodes [
  n1 = TextureSample(UVs=MaterialFunctionCall(Texture Scale=Constant() {r=0.8}) {materialFunction=&ScaleUVsByCenter}.UVs) {samplerSource=SSM_Clamp_WorldGroupSettings, texture=/FortniteCoreUI/UI/Quests/Textures/T_UI_PinQuest_Icon}
  n2 = Multiply(A=SmoothStep(Value=@n1.RGB) {constMax=0.02})
  n3 = MaterialFunctionCall(Alpha=@n1.RGB, Min=Constant() {r=0.1}, Max=Constant() {r=0.2}) {materialFunction=&SmoothStep}
  n4 = Constant() {r=0.1}
  n5 = Constant() {r=0.3}
  n6 = MaterialFunctionCall(Float2=TextureCoordinate()) {materialFunction=&BreakOutFloat2Components}
  n7 = MaterialFunctionCall(X=@n6.R, Y=Add(A=@n6.G) {constB=0.07}) {materialFunction=&MakeFloat2}
  n8 = Multiply(
   A=OneMinus(None=MaterialFunctionCall(Alpha=@n1.RGB, Min=@n4, Max=@n5) {materialFunction=&SmoothStep}.Result)
   B=MaterialFunctionCall(UVs=@n7.Result, Size=Constant() {r=0.4}) {materialFunction=&MF_UI_Circle}.Fill
  )
  LinearInterpolate(A=LinearInterpolate(A=Multiply(A=@n2) {constB=0.2}, B=@n3.Result, Alpha=@n3.Result), Alpha=@n8)
  MaterialFunctionCall(UVs=@n7.Result, Texture Scale=Constant() {r=0.4}) {materialFunction=&ScaleUVsByCenter}
  Multiply(A=OneMinus(None=@n8), B=@n3.Result)
 ]
}
