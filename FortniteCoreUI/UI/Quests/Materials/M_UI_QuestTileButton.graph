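# Translucent UI material for a quest tile button, providing the backplate look across idle, hovered, focused, pressed and disabled states.
# Slate post-buffer blur feeds CoreUI BackplateB and BackplateC functions, and a ModularButton lerp function blends the states, with hovered, focused and pressed sharing one input.
# A rotatable rounded mask shapes corners, with parameters for cut or squared-off variants.
# Low quality halves the disabled backplate's value.
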Material /FortniteCoreUI/UI/Quests/Materials/M_UI_QuestTileButton
graph {
 properties {
  bAutomaticallySetUsageInEditor=false
  bUsedWithStaticMesh=false
  blendMode=BLEND_Translucent
  materialDomain=MD_UI
  usedShadingModels=Default Lit
 }
 nodes [
  n1 = NamedRerouteUsage()
  MaterialFunctionCall!(
   Idle=NamedRerouteUsage()
   Idle Opacity
   Hovered=@n1
   Hovered Opacity
   Focused=@n1
   Focused Opacity
   Pressed=@n1
   Pressed Opacity
   Disabled=NamedRerouteUsage()
   Disabled Opacity
  ) {materialFunction=/FortUILibrary/Systems/ModularButton/MaterialFunctions/MF_UIKit_ButtonStatesBase_Lerp}
  n2 = Divide(A=ScalarParameter() {desc=Angle of rotation, parameterName=IndicatorDirection}) {constB=360}
  n3 = VectorParameter() {
   group=Shape
   parameterName=CornerRadius
   sortPriority=1
   channelNames {a=BottomRight, b=TopRight, g=BottomLeft, r=TopLeft}
   defaultValue=<rgba 16 16 16 16>
  }
  n4 = AppendVector(A=@n3, B=@n3)
  n5 = Constant() {r=1}
  n6 = MaterialFunctionCall(R=@n5, G=@n3.BottomRight, B=@n5, A=@n3.BottomRight) {materialFunction=&AppendMany}
  n7 = ScalarParameter() {desc=Whether or not the backplate is chopped, parameterName=IsCut}
  MaterialFunctionCall(
   UV=Rotator(Coordinate=TextureCoordinate(), Time=Multiply(A=@n2) {constB=-1}) {speed=6.2832}
   Corner Radius=LinearInterpolate(
    A=LinearInterpolate(A=@n4, B=@n6.RGBA, Alpha=@n7)
    B=Constant() {r=1}
    Alpha=ScalarParameter() {desc=All corner radiuses will be replaced with 1 if this is true, parameterName=IsSquaredOff}
   )
   Radius in Pixels=ScalarParameter() {defaultValue=1, group=Shape, parameterName=Radius in Pixels, sliderMax=1, sortPriority=2}
  ) {materialFunction=&MF_UI_CoreUI_RoundedMask}
  NamedRerouteDeclaration(None=MaterialFunctionCall!("UVs (Optional)") {materialFunction=/Engine/Functions/UserInterface/GetSlatePost0}.LinearRGB) {name=Blur RGB, nodeColor=<rgba 0 0 0 1>}
  n8 = MaterialFunctionCall!(Post Buffer Blur=NamedRerouteUsage()) {materialFunction=/FortUILibrary/MaterialFunctions/CoreUI/MF_UI_CoreUI_BackplateB}
  NamedRerouteDeclaration(None=@n8.Weak) {name=BackplateB_Weak, nodeColor=<rgba 0 0.2462 1 1>}
  NamedRerouteDeclaration(None=@n8.Result) {name=BackplateB, nodeColor=<rgba 0 0.2462 1 1>}
  n9 = MaterialFunctionCall!(Post Buffer Blur=NamedRerouteUsage()) {materialFunction=/FortUILibrary/MaterialFunctions/CoreUI/MF_UI_CoreUI_BackplateC}
  NamedRerouteDeclaration(None=QualitySwitch(Default=@n9.Result, Low=Multiply(A=@n9.Result) {constB=0.5})) {name=Disabled, nodeColor=<rgba 0 0.2462 1 1>}
 ]
}
