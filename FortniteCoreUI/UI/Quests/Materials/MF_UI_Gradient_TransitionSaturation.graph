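# Material function for quest UI materials that raises or lowers the saturation of a colored gradient within a masked area.
# A Desaturation node with neutral grey luminance factors takes the negated saturation input as its fraction, clamped to ±2 despite the -1 to 1 label.
# The saturated mask, capped by a threshold via Min, forms the blend space between original and adjusted color.
# The clamped result is RGB only, despite the RGBA output name.
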MaterialFunction /FortniteCoreUI/UI/Quests/Materials/MF_UI_Gradient_TransitionSaturation
graph {nodes [
  n1 = NamedRerouteUsage()
  n2 = ComponentMask(None=@n1) {b=true, desc=Because Desaturation output a Float3, g=true, r=true}
  n3 = Desaturation(None=@n1, Fraction=NamedRerouteUsage()) {
   desc="Must be grey 0.5,0.5,0.5 to work with \r\na maximum of colors without having to \r\nchange the saturation value"
   luminanceFactors=<rgba 0.5 0.5 0.5 1>
  }
  n4 = NamedRerouteUsage()
  FunctionOutput(None=Saturate(None=LinearInterpolate(A=@n2, B=@n3, Alpha=@n4))) {outputName=Colorful RGBA Gradient}
  n5 = FunctionInput() {
   bUsePreviewValueAsDefault=true
   description="Space where to apply the saturation (can be a gradient, a shape...)."
   inputName=Saturation Mask
   inputType=FunctionInput_Scalar
   sortPriority=4
  }
  n6 = FunctionInput() {
   bUsePreviewValueAsDefault=true
   description="Where to apply the saturation (can be a gradient, a shape...)."
   inputName=Saturation Mask Threshold
   inputType=FunctionInput_Scalar
   sortPriority=5
   previewValue=<0.95 0 0 1>
  }
  NamedRerouteDeclaration(None=Min(A=Saturate(None=@n5), B=Saturate(None=@n6))) {name=Saturation Space, nodeColor=<rgba 0 0 0 1>}
  n7 = FunctionInput() {bUsePreviewValueAsDefault=true, inputName="Saturation (-1 to 1)", inputType=FunctionInput_Scalar, sortPriority=3}
  NamedRerouteDeclaration(None=Multiply(A=Clamp(None=@n7) {maxDefault=2, minDefault=-2}) {constB=-1, desc=Because using desaturation node}) {name=Saturation, nodeColor=<rgba 0 0 0 1>}
  n8 = FunctionInput() {bUsePreviewValueAsDefault=true, description=Can be any Colored Gradient., inputName=Colored Gradient, sortPriority=1, previewValue=<0 0 0 0>}
  NamedRerouteDeclaration(None=@n8) {name=Colored Gradient Source, nodeColor=<rgba 0 0 0 1>}
 ]}
