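# Material function for UI (under Quests) that builds a colorful RGBA gradient from three colors with an adjustable, masked saturation effect.
# Colors A and B blend along a gradient input, then Color C overlays them through its space and display inputs.
# MF_UI_Gradient_TransitionSaturation runs before and after that overlay, with lerps selecting between results so the "Color C Covers Saturation" input controls whether C hides the effect.
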MaterialFunction /FortniteCoreUI/UI/Quests/Materials/MF_UI_Gradient_3Colors_WithSaturation
graph {nodes [
  n1 = NamedRerouteUsage()
  n2 = MaterialFunctionCall!(
   Colored Gradient=@n1
   "Saturation (-1 to 1)"=NamedRerouteUsage()
   Saturation Mask=NamedRerouteUsage()
   Saturation Mask Threshold=NamedRerouteUsage()
  ) {
   desc="Apply the saturation to the gradient A to B only if Color C doesn't cover the saturation\r\nDo not apply it if Color C covers the saturation because in that case it will be applied next to the 3 colors."
   materialFunction=&/MF_UI_Gradient_TransitionSaturation
  }
  n3 = LinearInterpolate(A=@n1, B=@n2.Colorful RGBA Gradient, Alpha=NamedRerouteUsage())
  n4 = LinearInterpolate(A=@n3, B=LinearInterpolate(A=@n3, B=NamedRerouteUsage(), Alpha=NamedRerouteUsage()), Alpha=NamedRerouteUsage())
  n5 = MaterialFunctionCall(
   Colored Gradient=@n4
   "Saturation (-1 to 1)"=NamedRerouteUsage()
   Saturation Mask=NamedRerouteUsage()
   Saturation Mask Threshold=NamedRerouteUsage()
  ) {
   desc="Apply the saturation to the gradient A to B only if Color C doesn't cover the saturation\r\nDo not apply it if Color C covers the saturation because in that case it will be applied next to the 3 colors."
   materialFunction=&MF_UI_Gradient_TransitionSaturation
  }
  FunctionOutput(None=LinearInterpolate(A=@n5.Colorful RGBA Gradient, B=@n4, Alpha=NamedRerouteUsage())) {outputName=Colorful RGBA Gradient}
  n6 = FunctionInput() {
   bUsePreviewValueAsDefault=true
   description="Where to apply the Color C (can be a gradient, a shape...)."
   inputName="Display Color C (0 to 1)"
   inputType=FunctionInput_Scalar
   sortPriority=6
  }
  NamedRerouteDeclaration(None=Saturate(None=@n6)) {name=Apply Colored Transition, nodeColor=<rgba 0 0 0 1>}
  n7 = FunctionInput() {bUsePreviewValueAsDefault=true, description=Color applied on the Color C Space, inputName=Color C, sortPriority=3, previewValue=<0 0 0 0>}
  NamedRerouteDeclaration(None=Saturate(None=@n7)) {name=Color C, nodeColor=<rgba 0 0 0 1>}
  n8 = FunctionInput() {
   bUsePreviewValueAsDefault=true
   description="Space where to apply the saturation (can be a gradient, a shape...)."
   inputName=Saturation Mask
   inputType=FunctionInput_Scalar
   sortPriority=8
  }
  NamedRerouteDeclaration(None=Saturate(None=@n8)) {name=Saturation Mask, nodeColor=<rgba 0 0 0 1>}
  n9 = FunctionInput() {bUsePreviewValueAsDefault=true, description=Color which merges with Color B, inputName=Color A, sortPriority=1, previewValue=<0 0 0 0>}
  n10 = FunctionInput() {bUsePreviewValueAsDefault=true, description=Color which merges with Color A, inputName=Color B, sortPriority=2, previewValue=<0 0 0 0>}
  n11 = FunctionInput() {
   bUsePreviewValueAsDefault=true
   description=The gradient on which to apply the Color A and Color B.
   inputName=A to B Gradient
   inputType=FunctionInput_Scalar
   sortPriority=4
  }
  NamedRerouteDeclaration(None=LinearInterpolate(A=Saturate(None=@n9), B=Saturate(None=@n10), Alpha=Saturate(None=@n11)) {desc=calculate the classic bicolor gradient}) {name=Colored Gradient Source, nodeColor=<rgba 0 0 0 1>}
  n12 = FunctionInput() {
   bUsePreviewValueAsDefault=true
   description="Where to apply the Color C (can be a gradient, a shape...)."
   inputName=Color C Space
   inputType=FunctionInput_Scalar
   sortPriority=5
  }
  NamedRerouteDeclaration(None=Saturate(None=@n12)) {name=Color C Space, nodeColor=<rgba 0 0 0 1>}
  NamedRerouteDeclaration(None=FunctionInput() {bUsePreviewValueAsDefault=true, inputName="Saturation (-1 to 1)", inputType=FunctionInput_Scalar, sortPriority=7}) {name=Saturation, nodeColor=<rgba 0 0 0 1>}
  n13 = FunctionInput() {
   bUsePreviewValueAsDefault=true
   description="When 1, the color C will cover the saturation effect applied between the colors A and B."
   inputName="Color C Covers Saturation (0 to 1)"
   inputType=FunctionInput_Scalar
   sortPriority=10
  }
  NamedRerouteDeclaration(None=Saturate(None=@n13)) {name=Color C covers Saturarion, nodeColor=<rgba 0 0.0855 1 1>}
  n14 = FunctionInput() {
   bUsePreviewValueAsDefault=true
   description="Where to apply the saturation (can be a gradient, a shape...)."
   inputName=Saturation Mask Threshold
   inputType=FunctionInput_Scalar
   sortPriority=9
   previewValue=<0.95 0 0 1>
  }
  NamedRerouteDeclaration(None=@n14) {name=Saturation Mask Threshold, nodeColor=<rgba 0 0 0 1>}
 ]}
