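# Translucent UI material that draws the quest tile background, with an item image over a color gradient.
# A three-color gradient, corner gradient and scaled, offset texture are composited with rounded-corner masking; Slate post-buffer samples supply backplate and debug colors.
# Scalar parameters blend between header, tracked, notification, side-nav and Quest 3.0 looks, including a vertex-transform resize.
# Custom nodes read the viewport UI scale to adjust image offset, and a static switch selects debug display.
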Material /FortniteCoreUI/UI/Quests/Materials/M_UI_QuestTile_Background
graph {
 properties {
  bAutomaticallySetUsageInEditor=false
  bUsedWithStaticMesh=false
  blendMode=BLEND_Translucent
  materialDomain=MD_UI
  usedShadingModels=Default Lit
 }
 nodes [
  n1 = MaterialFunctionCall!(Anchor, Translation, Padding, Scale) {materialFunction=/FortUILibrary/MaterialFunctions/Basics/MF_UI_VertexTransform}
  n2 = MaterialFunctionCall(Translation=Constant2Vector() {g=-12}, Padding=Constant2Vector() {g=24}) {materialFunction=&MF_UI_VertexTransform}
  LinearInterpolate(A=@n1.Transformed Position, B=@n2.Transformed Position, Alpha=NamedRerouteUsage())
  n3 = NamedRerouteUsage()
  n4 = NamedRerouteUsage()
  n5 = Multiply(
   A=MaterialFunctionCall(GradientStart=Constant() {r=0.35}, GradientEnd=Constant() {r=0.65}, RotationAngle=Constant() {r=-0.02}) {materialFunction=&MF_UI_LinearGradient}.Result
   B=LinearInterpolate(Alpha=MaterialFunctionCall() {materialFunction=&MF_UI_LinearGradient}.Result) {constA=1, constB=0.25}
  )
  n6 = ScalarParameter() {group=BP logic, parameterName=IsHeader}
  n7 = If(A=NamedRerouteUsage(), "A > B"=Constant(), "A == B"=@n6, "A < B"=@n6)
  Max(A=Multiply(A=Multiply(A=@n3, B=@n4), B=LinearInterpolate(B=@n5, Alpha=@n7) {constA=1, constAlpha=1}), B=NamedRerouteUsage())
  n8 = MaterialFunctionCall() {materialFunction=&GetUserInterfaceUV}
  NamedRerouteDeclaration(None=@n8.Pixel Size) {name=PixelSize, nodeColor=<rgba 0.0262 1 0 1>}
  n9 = LinearInterpolate(
   A=ScalarParameter() {group=2 - Texture, parameterName=Texture_YOffset}
   B=ScalarParameter() {group=4 - Quest 3.0, parameterName=Quest3.0_Texture_YOffset}
   Alpha=NamedRerouteUsage()
  )
  n10 = ComponentMask(None=NamedRerouteUsage()) {g=true}
  NamedRerouteDeclaration(None=Divide(A=Multiply(A=@n9) {constB=-1}, B=Max(A=@n10))) {name=Texture_YOffset, nodeColor=<rgba 0.0257 1 0 1>}
  n11 = TextureObjectParameter() {group=2 - Texture, parameterName=Texture, texture=/FortniteCoreUI/UI/Quests/Textures/T_UI_Figment_S07_Kickstart_MincePounce}
  NamedRerouteDeclaration(None=TextureProperty(Tex=@n11)) {name=Texture_Size, nodeColor=<rgba 0.0257 1 0 1>}
  NamedRerouteDeclaration(None=@n11) {name=Texture_Object, nodeColor=<rgba 0 0.2461 1 1>}
  n12 = NamedRerouteUsage()
  n13 = Add(A=Multiply(A=OneMinus(None=@n12)) {constB=1.5})
  n14 = AppendVector(A=ComponentMask(None=@n13) {g=true}, B=ComponentMask(None=@n13) {r=true})
  n15 = NamedRerouteUsage()
  n16 = Constant() {desc="Please don't use this\r\nBetter use Texture_XPixelOffset", r=0.5}
  n17 = Constant() {desc="Please don't use this\r\nBetter use Texture_YPixelOffset", r=0.5}
  n18 = NamedRerouteUsage()
  n19 = Divide(
   A=ComponentMask(None=NamedRerouteUsage()) {g=true}
   B=Multiply(
    A=Constant() {r=134}
    B=Custom() {code="return View.ViewportScaleUI;", description=ViewportScaleFactor, outputType=CMOT_Float1, showCode=true, inputs []}
   )
  ) {constA=144, constB=114}
  n20 = ScalarParameter() {defaultValue=1, parameterName=UseAlternativeHeight}
  n21 = Add(A=ScalarParameter() {group=BP logic, parameterName=IsHeader}, B=If(A=NamedRerouteUsage(), "A > B"=Constant(), "A == B"=@n20, "A < B"=@n20))
  n22 = AppendVector(
   A=LinearInterpolate(A=NamedRerouteUsage(), Alpha=ScalarParameter() {parameterName=IsNotification}) {constB=0}
   B=LinearInterpolate(A=@n18, B=Multiply(A=@n19, B=@n18), Alpha=Saturate(None=@n21))
  )
  n23 = Custom() {code="return View.ViewportScaleUI;", description=ViewportScaleFactor, outputType=CMOT_Float1, showCode=true, inputs []}
  n24 = MaterialFunctionCall!(
   Texture Scale=LinearInterpolate(A=@n15, B=Multiply(A=@n14, B=@n15), Alpha=NamedRerouteUsage())
   UVs=Add(A=NamedRerouteUsage(), B=Multiply(A=@n22, B=@n23))
   Offset XY Texturesize Z=AppendVector(A=AppendVector(A=@n16, B=@n17), B=Constant() {r=0.5})
  ) {materialFunction=/Engine/Functions/Engine_MaterialFunctions01/Texturing/ScaleUVsAroundPoint}
  n25 = NamedRerouteUsage()
  n26 = Constant() {r=1}
  n27 = TextureSample(
   UVs=NamedRerouteDeclaration(None=MaterialFunctionCall(UV=@n24.UVs, TextureSize=@n25, ScalePivotPoint=@n26) {materialFunction=&MF_UI_UVScaling}.ScaleY) {name=TextureUVs, nodeColor=<rgba 0 0.2462 1 1>}
   Tex=NamedRerouteUsage()
  ) {samplerSource=SSM_Clamp_WorldGroupSettings}
  n28 = Constant() {r=1}
  n29 = MaterialFunctionCall!(UV=NamedRerouteUsage(), Grid Size=@n28, Index=@n28) {materialFunction=/FortUILibrary/MaterialFunctions/Basics/MF_UI_IsolateUV}
  NamedRerouteDeclaration(None=Multiply(A=@n27.RGB, B=@n29.Result)) {name=Texture_RGB, nodeColor=<rgba 0 0.2461 1 1>}
  n30 = ScalarParameter() {defaultValue=1, desc="1 = texture fully opaque\r\n0 = texture completely transparent", group=2 - Texture, parameterName=ShowItemImage}
  n31 = Multiply(A=LinearInterpolate(A=@n30, B=Ceil(None=@n30), Alpha=NamedRerouteUsage()), B=Multiply(A=@n27.A, B=@n29.Result))
  n32 = Multiply(A=@n31, B=NamedRerouteUsage())
  n33 = NamedRerouteUsage()
  n34 = MaterialFunctionCall(GradientStart=Constant(), GradientEnd=Constant() {r=0.2}, RotationAngle=Constant() {r=0.4}, GradientOffset=Constant() {r=-0.2}) {materialFunction=&MF_UI_LinearGradient}
  NamedRerouteDeclaration(None=LinearInterpolate(
    A=LinearInterpolate(A=@n32, B=@n31, Alpha=@n33)
    B=Multiply(A=@n31, B=@n34.Result)
    Alpha=ScalarParameter() {group=BP logic, parameterName=IsHeader}
   )) {name=Texture_Alpha, nodeColor=<rgba 0 0.2462 1 1>}
  n35 = ScalarParameter() {
   defaultValue=1.9
   desc="Gradient softness\r\n0 = hard, sharp edge \r\n1 = smooth and soft"
   group=CornerGradient
   parameterName=CornerGradient_SmoothStepMax
  }
  NamedRerouteDeclaration(None=@n35) {name=CornerGradient_SmoothStepMax, nodeColor=<rgba 0.0257 1 0 1>}
  n36 = ScalarParameter() {defaultValue=3, desc="Don't use this please\r\nBetter use CornerGradient_Size", group=Legacy, parameterName=CornerGradient_XPosition}
  n37 = ScalarParameter() {desc="Don't use this please\r\nBetter use CornerGradient_Size", group=Legacy, parameterName=CornerGradient_YPosition}
  n38 = ScalarParameter() {defaultValue=4, group=CornerGradient, parameterName=CornerGradient_Size}
  n39 = MaterialFunctionCall(
   X=Constant() {desc="Don't use this please\r\nBetter use CornerGradient_Size", r=3}
   Y=Constant() {desc="Don't use this please\r\nBetter use CornerGradient_Size", r=0.4}
   Z=Constant() {r=3}
  ) {materialFunction=&MakeFloat3}
  n40 = ScalarParameter() {parameterName=IsTracked}
  NamedRerouteDeclaration(None=LinearInterpolate(
    A=MaterialFunctionCall(X=@n36, Y=@n37, Z=@n38) {materialFunction=&MakeFloat3}.Result
    B=LinearInterpolate(A=@n39.Result, Alpha=@n40) {constB=0}
    Alpha=NamedRerouteUsage()
   ) {constB=0}) {name=CornerGradient_AnchorPosition, nodeColor=<rgba 0 0.2462 1 1>}
  n41 = NamedRerouteUsage()
  n42 = Constant() {r=0.3}
  n43 = AppendVector(A=@n42, B=@n42)
  n44 = NamedRerouteUsage()
  n45 = MaterialFunctionCall(UV Channel=MaterialFunctionCall(
    UVs=MaterialFunctionCall(Texture Scale=@n43, Offset XY Texturesize Z=@n44) {materialFunction=&ScaleUVsAroundPoint}.UVs
    Rotation Center=Constant() {r=3}
    "Rotation Angle (0-1)"=Constant() {r=-0.1}
   ) {materialFunction=&CustomRotator}.Rotated Values) {materialFunction=&LinearGradient}
  NamedRerouteDeclaration(None=Multiply(A=SmoothStep(Max=@n41, Value=@n45.UGradient) {constMax=1.5}, B=NamedRerouteUsage())) {name=CornerGradient_Mask, nodeColor=<rgba 0 0.2461 1 1>}
  n46 = VectorParameter() {group=CornerGradient, parameterName=CornerGradient_ColorFade, defaultValue=<rgba 0.0482 0.1329 0.0194 1>}
  NamedRerouteDeclaration(None=@n46.RGB) {name=CornerGradient_RGB, nodeColor=<rgba 0 0.2461 1 1>}
  n47 = VectorParameter() {desc=LeftColor, group=1 - Background, parameterName=BackgroundColor, defaultValue=<rgba 0.007 0.1221 0.1329 1>}
  NamedRerouteDeclaration(None=@n47.RGB) {name=Background_Color, nodeColor=<rgba 1 0.9316 0.0156 1>}
  n48 = ScalarParameter() {group=BP logic, parameterName=IsHeader}
  NamedRerouteDeclaration(None=Multiply(A=ComponentMask(None=@n46.A) {a=true}, B=OneMinus(None=@n48))) {name=CornerGradient_Alpha, nodeColor=<rgba 0 0.2461 1 1>}
  n49 = VectorParameter() {group=Debug, parameterName=BackgroundDefaultColor, defaultValue=<rgba 0.099 0.099 0.099 0.5>}
  NamedRerouteDeclaration(None=@n49.RGB) {name=Background_DebugColor, nodeColor=<rgba 1 0.9316 0.0156 1>}
  NamedRerouteDeclaration(None=@n49.A) {name=Background_DebugAlpha, nodeColor=<rgba 1 0.9316 0.0156 1>}
  n50 = MaterialFunctionCall!("UVs (Optional)") {materialFunction=/Engine/Functions/UserInterface/GetSlatePost0}
  n51 = MaterialFunctionCall!(Color=Max(A=Saturate(None=@n50.LinearRGB)) {constB=0.0001}, Contrast=ScalarParameter() {defaultValue=0.4, parameterName=EffectWeakness}) {materialFunction=/Engine/Functions/Engine_MaterialFunctions01/ImageAdjustment/Contrast_Preserve_Color}
  NamedRerouteDeclaration(None=MaterialFunctionCall!(Blend=Saturate(None=@n51.Result), Base=NamedRerouteUsage()) {materialFunction=/Engine/Functions/Engine_MaterialFunctions03/Blends/Blend_Screen}.Result) {name=Background_DebugRGB, nodeColor=<rgba 0 0.2461 1 1>}
  n52 = NamedRerouteUsage()
  n53 = NamedRerouteUsage()
  n54 = ScalarParameter() {defaultValue=3, group=SideNavigationStyle, parameterName=SN_Scale}
  n55 = Multiply(A=@n54, B=Divide(A=ComponentMask(None=@n53) {r=true}, B=ComponentMask(None=@n53) {g=true}))
  n56 = Divide(A=AppendVector(
    A=ScalarParameter() {defaultValue=57, group=SideNavigationStyle, parameterName=SN_OffsetX}
    B=ScalarParameter() {defaultValue=13, group=SideNavigationStyle, parameterName=SN_OffsetY}
   )) {constB=100}
  n57 = Constant() {r=1}
  n58 = TextureSample(
   UVs=MaterialFunctionCall(Texture Scale=AppendVector(A=@n55, B=@n54), UVs=NamedRerouteUsage(), Offset XY Texturesize Z=AppendVector(A=@n56, B=@n57)) {materialFunction=&ScaleUVsAroundPoint}.UVs
   Tex=NamedRerouteUsage()
  ) {samplerSource=SSM_Clamp_WorldGroupSettings}
  n59 = ScalarParameter() {desc=Only for debugging. Should always to be 0, group=SideNavigationStyle, parameterName=IsSideNav}
  n60 = NamedRerouteUsage()
  n61 = ScalarParameter() {defaultValue=1, desc="1 = texture fully opaque\r\n0 = texture completely transparent", group=2 - Texture, parameterName=ShowItemImage}
  n62 = Ceil(None=@n61)
  n63 = NamedRerouteUsage()
  n64 = LinearInterpolate(A=NamedRerouteUsage(), B=Multiply(A=Multiply(A=@n58.A, B=@n60), B=LinearInterpolate(A=@n61, B=@n62, Alpha=@n63)), Alpha=@n59) {constB=2}
  n65 = LinearInterpolate(
   A=LinearInterpolate(A=NamedRerouteUsage(), B=LinearInterpolate(A=@n52, B=@n58.RGB, Alpha=@n59) {constB=2}, Alpha=@n64) {constAlpha=0}
   B=NamedRerouteUsage()
   Alpha=LinearInterpolate(A=NamedRerouteUsage(), Alpha=@n59) {constB=0}
  )
  n66 = MaterialFunctionCall(GradientStart=Constant() {r=0.4}, GradientEnd=Constant() {r=1.3}, RotationAngle=Constant() {r=0.75}) {materialFunction=&MF_UI_LinearGradient}
  n67 = LinearInterpolate(B=Constant3Vector() {constant=<rgba 0.0048 0.004 0.0116 1>}, Alpha=Saturate(None=SmoothStep(Value=@n66.Result) {constMin=0.1, constValue=1})) {constA=1}
  n68 = NamedRerouteUsage()
  n69 = ScalarParameter() {group=BP logic, parameterName=IsHeader}
  NamedRerouteDeclaration(None=LinearInterpolate(A=Multiply(A=@n65, B=@n67), B=@n65, Alpha=Max(A=@n68, B=@n69))) {name=CompositeBackground_RGB, nodeColor=<rgba 0 0.2462 1 1>}
  n70 = NamedRerouteUsage()
  n71 = NamedRerouteUsage()
  n72 = NamedRerouteUsage()
  n73 = ScalarParameter() {defaultValue=0.7, desc="GradientMidOffset\r\nGradientStart < GradientMid < GradientEnd", group=1 - Background, parameterName=2 GradientMid}
  n74 = ScalarParameter() {desc=From 0 to 1, group=1 - Background, parameterName=4 GradientRotation}
  n75 = ScalarParameter() {desc=Full gradient offset, group=1 - Background, parameterName=5 GradientOffset}
  n76 = MaterialFunctionCall(
   GradientStart=ScalarParameter() {desc="Gradient Start Offset \r\nGradientStart < GradientMid < GradientEnd", group=1 - Background, parameterName=1 GradientStart}
   GradientEnd=@n73
   RotationAngle=@n74
   GradientOffset=@n75
  ) {materialFunction=&MF_UI_LinearGradient}
  n77 = ScalarParameter() {defaultValue=1.6, desc="Gradient End Offset\r\nGradientStart < GradientMid < GradientEnd", group=1 - Background, parameterName=3 GradientEnd}
  n78 = MaterialFunctionCall(GradientStart=@n73, GradientEnd=@n77, RotationAngle=@n74, GradientOffset=@n75) {materialFunction=&MF_UI_LinearGradient}
  n79 = Multiply(A=Subtract(A=LinearInterpolate(Alpha=@n78.Result) {constA=0.5}) {constB=0.5}) {constB=2}
  n80 = Constant() {r=1}
  n81 = Constant() {r=1}
  n82 = Constant() {r=0.95}
  n83 = Constant() {r=1}
  n84 = LinearInterpolate(
   A=LinearInterpolate(
    B=MaterialFunctionCall!(
     Color A=@n70
     Color B=@n71
     Color C=@n72
     A to B Gradient=@n76.Result
     Color C Space=@n79
     "Display Color C (0 to 1)"=@n80
     "Saturation (-1 to 1)"=@n81
     Saturation Mask
     Saturation Mask Threshold=@n82
     "Color C Covers Saturation (0 to 1)"=@n83
    ) {materialFunction=&/MF_UI_Gradient_3Colors_WithSaturation}.Colorful RGBA Gradient
    Alpha=NamedRerouteUsage()
   )
   B=@n70
   Alpha=ScalarParameter() {group=BP logic, parameterName=IsHeader}
  )
  n85 = NamedRerouteUsage()
  n86 = NamedRerouteUsage()
  NamedRerouteDeclaration(None=LinearInterpolate(A=@n84, B=LinearInterpolate(A=@n85, B=@n84, Alpha=@n86), Alpha=NamedRerouteUsage())) {name=Background_RGB, nodeColor=<rgba 0 0.2462 1 1>}
  n87 = MaterialFunctionCall() {materialFunction=&GetUserInterfaceUV}
  n88 = LinearInterpolate(A=@n87.Normalized UV, B=Divide(A=OneMinus(None=@n87.Normalized UV), B=NamedRerouteUsage()), Alpha=NamedRerouteUsage()) {constA=1}
  n89 = LinearInterpolate(A=Constant() {r=16}, B=Constant() {r=12}, Alpha=NamedRerouteUsage())
  n90 = Constant() {r=1}
  n91 = Constant() {r=1}
  n92 = ScalarParameter() {defaultValue=0.25, group=SideNavigationStyle, parameterName=SideNav_CornerRadius}
  n93 = ScalarParameter() {group=SideNavigationStyle, parameterName=SideNav_RadiusInPixels}
  NamedRerouteDeclaration(None=LinearInterpolate(
    A=MaterialFunctionCall(UV=@n88, Size=@n91, Corner Radius=@n89, Radius in Pixels=@n90) {materialFunction=&MF_UI_CoreUI_RoundedMask}.Shape Mask
    B=MaterialFunctionCall(UV=@n88, Size=@n91, Corner Radius=@n92, Radius in Pixels=@n93) {materialFunction=&MF_UI_CoreUI_RoundedMask}.Shape Mask
    Alpha=ScalarParameter() {desc=Only for debugging. Should always to be 0, group=SideNavigationStyle, parameterName=IsSideNav}
   ) {constB=2}) {name=RoundedFill_Mask, nodeColor=<rgba 0 0.2462 1 1>}
  NamedRerouteDeclaration(None=@n8.9-Slice UV) {name=UVs, nodeColor=<rgba 0.0262 1 0 1>}
  n94 = VectorParameter() {desc=CenterColor, group=1 - Background, parameterName=ColorSecond, defaultValue=<rgba 0.1413 0.5149 0.0065 1>}
  NamedRerouteDeclaration(None=@n94.RGB) {name=Background_ColorSecond, nodeColor=<rgba 1 0.9316 0.0156 1>}
  n95 = VectorParameter() {desc=RightColor, group=1 - Background, parameterName=ColorCenter, defaultValue=<rgba 0.3095 1 0.0137 1>}
  NamedRerouteDeclaration(None=@n95.RGB) {name=Background_ColorCenter, nodeColor=<rgba 1 0.9316 0.0156 1>}
  n96 = ScalarParameter() {
   defaultValue=0.7
   desc="0 = completely transparent \r\n1 = fully opaque "
   group=1 - Background
   parameterName=Background Opacity
   sliderMax=1
  }
  n97 = LinearInterpolate(A=@n96, Alpha=NamedRerouteUsage())
  n98 = ScalarParameter() {group=BP logic, parameterName=IsHeader}
  NamedRerouteDeclaration(None=StaticSwitchParameter(True=NamedRerouteUsage(), False=LinearInterpolate(A=@n97, Alpha=@n98)) {parameterName=IsDebugDisplay}) {name=Background_Opacity, nodeColor=<rgba 0 0.2461 1 1>}
  NamedRerouteDeclaration(None=@n64) {name=Alpha, nodeColor=<rgba 0 0.1328 1 1>}
  NamedRerouteDeclaration(None=Divide(A=@n1.New Pixel Size, B=@n2.New Pixel Size)) {name=QuestV3NewPixelSize, nodeColor=<rgba 0.0547 0 1 1>}
  n99 = ScalarParameter() {defaultValue=0.4, group=4 - Quest 3.0, parameterName=Quest3.0_GradientStart}
  n100 = Constant() {r=1}
  n101 = Constant()
  NamedRerouteDeclaration(None=LinearInterpolate(
    A=MaterialFunctionCall(GradientStart=@n99, GradientEnd=@n100, RotationAngle=@n101) {materialFunction=&MF_UI_LinearGradient}.Result
    Alpha=ScalarParameter() {parameterName=IsTracked}
   ) {constB=0}) {name=Gradient_Quest3.0, nodeColor=<rgba 0 1 0.2266 1>}
  NamedRerouteDeclaration(None=MaterialFunctionCall() {materialFunction=&GetSlatePost0}.LinearRGB) {name=Blur RGB, nodeColor=<rgba 0 0 0 1>}
  NamedRerouteDeclaration(None=MaterialFunctionCall!(Post Buffer Blur=NamedRerouteUsage()) {materialFunction=/FortUILibrary/MaterialFunctions/CoreUI/MF_UI_CoreUI_BackplateB}.Weak) {name=BackplateB_Weak, nodeColor=<rgba 0 0.2462 1 1>}
  NamedRerouteDeclaration(None=ScalarParameter() {group=4 - Quest 3.0, parameterName=QuestV3}) {name=QuestV3, nodeColor=<rgba 0.0547 1 0 1>}
  NamedRerouteDeclaration(None=@n2.Total Scale Factor) {name=NewPixelSixe, nodeColor=<rgba 0 1 0.0156 1>}
  NamedRerouteDeclaration(None=LinearInterpolate(
    A=ScalarParameter() {defaultValue=1.05, group=2 - Texture, parameterName=Texture_Scale}
    B=ScalarParameter() {defaultValue=1.05, group=4 - Quest 3.0, parameterName=Quest3.0_Texture_Scale}
    Alpha=NamedRerouteUsage()
   )) {name=Texture_Scale, nodeColor=<rgba 0.0262 1 0 1>}
  n102 = LinearInterpolate(
   A=ScalarParameter() {defaultValue=-50, group=2 - Texture, parameterName=Texture_XOffset, sliderMax=100, sliderMin=-100}
   B=ScalarParameter() {defaultValue=-10, group=4 - Quest 3.0, parameterName=Quest3.0_Texture_XOffset, sliderMax=100, sliderMin=-100}
   Alpha=NamedRerouteUsage()
  )
  n103 = ComponentMask(None=NamedRerouteUsage()) {r=true}
  NamedRerouteDeclaration(None=Divide(A=Multiply(A=@n102) {constB=-1}, B=Max(A=@n103))) {name=Texture_XOffset, nodeColor=<rgba 0.0257 1 0 1>}
  StaticSwitchParameter(True=NamedRerouteUsage(), False=NamedRerouteUsage()) {parameterName=IsDebugDisplay}
 ]
}
