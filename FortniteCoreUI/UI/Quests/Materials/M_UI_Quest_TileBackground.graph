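# Translucent UI material for quest tile backgrounds, coloured by button interaction and selection state.
# Modular button functions lerp state colour and opacity, with an optional hidden idle state.
# Slate post-buffer 0, contrast-adjusted with colour preserved, is screen-blended onto that colour.
# Opacity is shaped by rounded masks; an optional cut uses a rotated, pixel-width UV split to switch corner rounding along a directional indicator strip.
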Material /FortniteCoreUI/UI/Quests/Materials/M_UI_Quest_TileBackground
graph {
 properties {
  bUsedWithStaticMesh=false
  blendMode=BLEND_Translucent
  materialDomain=MD_UI
  pixelDepthOffsetMode=PDOM_Legacy
  usedShadingModels=Default Lit
 }
 nodes [
  n1 = Max(A=Saturate(None=MaterialFunctionCall!("UVs (Optional)") {materialFunction=/Engine/Functions/UserInterface/GetSlatePost0}.LinearRGB)) {constB=0.0001}
  n2 = ScalarParameter() {defaultValue=0.4, parameterName=EffectWeakness}
  n3 = MaterialFunctionCall!(Idle, Idle Opacity, Hovered, Hovered Opacity, Focused, Focused Opacity, Pressed, Pressed Opacity, Disabled, Disabled Opacity) {materialFunction=/FortUILibrary/Systems/ModularButton/MaterialFunctions/MF_UIKit_ButtonStatesBase_Lerp}
  n4 = MaterialFunctionCall!(
   Base States=@n3.Color
   Base States Opacity=LinearInterpolate(A=@n3.Opacity, Alpha=ScalarParameter() {parameterName=bHideIdleState}) {constB=0}
   Selected_Idle
   Selected_Idle Opacity
   Selected_Hovered
   Selected_Hovered Opacity
   Selected_Focused
   Selected_Focused Opacity
   Selected_Pressed
   Selected_Pressed Opacity
   Selected_Disabled
   Selected_Disabled Opacity
  ) {materialFunction=/FortUILibrary/Systems/ModularButton/MaterialFunctions/MF_UIKit_ButtonStatesSelectable_Lerp}
  MaterialFunctionCall!(
   Blend=Saturate(None=MaterialFunctionCall!(Color=@n1, Contrast=@n2) {materialFunction=/Engine/Functions/Engine_MaterialFunctions01/ImageAdjustment/Contrast_Preserve_Color}.Result)
   Base=@n4.Color
  ) {materialFunction=/Engine/Functions/Engine_MaterialFunctions03/Blends/Blend_Screen}
  n5 = MaterialFunctionCall(Radius in Pixels=Constant() {r=1}) {materialFunction=&MF_UI_CoreUI_RoundedMask}
  n6 = Divide(A=ScalarParameter() {defaultValue=90, desc=Angle, parameterName=IndicatorDirection}) {constB=360}
  n7 = ScalarParameter() {defaultValue=20, parameterName=Width Indicator}
  n8 = Abs(None=Cosine(None=@n6))
  n9 = Add(
   A=Rotator(Coordinate=TextureCoordinate(), Time=Multiply(A=@n6) {constB=-1}) {speed=6.2832}
   B=Multiply(A=MaterialFunctionCall!(Value PX=@n7, UV Pixel Size, X or Y=@n8) {materialFunction=/FortUILibrary/MaterialFunctions/Basics/MF_UI_PixelsToNormalized}.Normalized) {constB=-1}
  )
  n10 = Saturate(None=Ceil(None=ComponentMask(None=@n9) {g=true}))
  n11 = MaterialFunctionCall(Corner Radius=ScalarParameter() {defaultValue=20, parameterName=Corner Radius}, Radius in Pixels=Constant() {r=1}) {materialFunction=&MF_UI_CoreUI_RoundedMask}
  Multiply(
   A=@n4.Opacity
   B=LinearInterpolate(
    A=@n11.Shape Mask
    B=LinearInterpolate(A=@n5.Shape Mask, B=@n11.Shape Mask, Alpha=@n10) {constA=1}
    Alpha=ScalarParameter() {defaultValue=1, parameterName=IsCut}
   )
  )
 ]
}
