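# Material function that builds a per-corner rounded box signed distance field for UI materials, exposed in the SDFs library category.
# Normalized UVs are centered and scaled by widget pixel size, then a custom HLSL node computes the distance using box size and corner radii, optionally blended toward half-height rounding.
# Smoothsteps over glow ranges produce fill and outline masks, output alongside both raw distances.
# Radii, thickness and glow inputs follow viewport UI scale.
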MaterialFunction /FortUILibrary/MaterialFunctions/Shapes/SDFs/MF_SDF_NormalizedRoundedBox
graph {
 properties {bExposeToLibrary=true, libraryCategoriesText [SDFs]}
 nodes [
  FunctionOutput(None=NamedRerouteUsage()) {outputName=SDF, sortPriority=2}
  n1 = NamedRerouteUsage()
  n2 = NamedRerouteUsage()
  n3 = NamedRerouteUsage()
  FunctionOutput(None=OneMinus(None=SmoothStep(Min=@n1, Max=@n2, Value=@n3))) {outputName=Shape Fill}
  n4 = Abs(None=NamedRerouteUsage())
  n5 = NamedRerouteUsage()
  FunctionOutput(None=NamedRerouteDeclaration(None=Subtract(A=@n4, B=@n5) {constB=0.1}) {name=SDF outline, nodeColor=<rgba 0.7053 0 1 1>}) {outputName=SDF Outline, sortPriority=3}
  n6 = NamedRerouteUsage()
  n7 = NamedRerouteUsage()
  n8 = NamedRerouteUsage()
  FunctionOutput(None=OneMinus(None=SmoothStep(Min=@n6, Max=@n7, Value=@n8) {constMax=3})) {outputName=Shape Outline, sortPriority=1}
  n9 = FunctionInput(Preview=Constant2Vector()) {bUsePreviewValueAsDefault=true, inputName=Padding, inputType=FunctionInput_Vector2, sortPriority=1}
  Multiply(A=@n9, B=Custom() {code="return View.ViewportScaleUI;", description=ViewportScaleFactor, outputType=CMOT_Float1, inputs []})
  NamedRerouteDeclaration(None=FunctionInput(Preview=TextureCoordinate() {coordinateIndex=4}) {bUsePreviewValueAsDefault=true, inputName=NormalizedUVs, inputType=FunctionInput_Vector2}) {name=NormalizedUVs, nodeColor=<rgba 0 0.2462 1 1>}
  n10 = FunctionInput(Preview=Constant4Vector() {constant=<rgba 10 10 10 10>}.RGBA) {bUsePreviewValueAsDefault=true, inputName=BoxCornerRoundness, inputType=FunctionInput_Vector4, sortPriority=15}
  n11 = Custom() {code="return View.ViewportScaleUI;", description=ViewportScaleFactor, outputType=CMOT_Float1, inputs []}
  n12 = Subtract(
   A=ComponentMask(None=MaterialFunctionCall() {materialFunction=&GetUserInterfaceUV}.Pixel Size) {g=true}
   B=ComponentMask(None=NamedRerouteUsage()) {g=true}
  )
  n13 = Constant()
  NamedRerouteDeclaration(None=LinearInterpolate(
    A=Multiply(A=@n10, B=@n11)
    B=MaterialFunctionCall(R=@n12, G=@n12, B=@n12, A=@n12) {materialFunction=&AppendMany}.RGBA
    Alpha=FunctionInput(Preview=@n13) {bUsePreviewValueAsDefault=true, inputName=UseHalfHeight, inputType=FunctionInput_Scalar, sortPriority=50}
   )) {name=Box Corner Roundness, nodeColor=<rgba 0.3359 1 0 1>}
  n14 = MaterialFunctionCall() {materialFunction=&GetUserInterfaceUV}
  n15 = Subtract(A=NamedRerouteUsage()) {constB=0.5}
  NamedRerouteDeclaration(None=Multiply(A=NamedRerouteDeclaration(None=@n14.Pixel Size) {name=Widget Size, nodeColor=<rgba 0 1 0.5781 1>}, B=Multiply(A=@n15) {constB=2})) {name=UV Screen Space, nodeColor=<rgba 0 0.3672 1 1>}
  NamedRerouteDeclaration(None=Subtract(A=NamedRerouteUsage(), B=NamedRerouteUsage())) {name=Box Size, nodeColor=<rgba 0 0.625 1 1>}
  MaterialExpressionNamedRerouteUsage_56 = NamedRerouteUsage()
  MaterialExpressionNamedRerouteUsage_55 = NamedRerouteUsage()
  MaterialExpressionComponentMask_0 = ComponentMask(None=NamedRerouteUsage()) {a=true, b=true, g=true, r=true}
  n16 = Custom!(p=@MaterialExpressionNamedRerouteUsage_56, b=@MaterialExpressionNamedRerouteUsage_55, r=@MaterialExpressionComponentMask_0) {
   code="r.xy = (p.x>0.0)?r.xy : r.zw;\r\nr.x  = (p.y>0.0)?r.x  : r.y;\r\nfloat2 q = abs(p)-b+r.x;\r\nreturn min(max(q.x,q.y),0.0) + length(max(q,0.0)) - r.x;"
   description=SDF_RoundedBox
   outputType=CMOT_Float1
   inputs [
    {
     inputName=p
     input {expression=&self:MaterialExpressionNamedRerouteUsage_56}
    }
    {
     inputName=b
     input {expression=&self:MaterialExpressionNamedRerouteUsage_55}
    }
    {inputName=r, input {expression=&self:MaterialExpressionComponentMask_0}}
   ]
  }
  NamedRerouteDeclaration(None=@n16) {name=SDF, nodeColor=<rgba 0.7053 0 1 1>}
  n17 = Constant()
  NamedRerouteDeclaration(None=Multiply(
    A=FunctionInput(Preview=@n17) {bUsePreviewValueAsDefault=true, inputName=Fill Glow Min, inputType=FunctionInput_Scalar, sortPriority=2}
    B=Custom!() {code="return View.ViewportScaleUI;", description=ViewportScaleFactor, outputType=CMOT_Float1, inputs []}
   )) {name=Fill Glow Min, nodeColor=<rgba 0 1 0.0391 1>}
  n18 = Constant() {r=3}
  NamedRerouteDeclaration(None=Multiply(
    A=FunctionInput(Preview=@n18) {bUsePreviewValueAsDefault=true, inputName=Fill Glow Max, inputType=FunctionInput_Scalar, sortPriority=3}
    B=Custom() {code="return View.ViewportScaleUI;", description=ViewportScaleFactor, outputType=CMOT_Float1, inputs []}
   )) {name=Fill Glow Max, nodeColor=<rgba 0 1 0.0391 1>}
  n19 = Constant()
  NamedRerouteDeclaration(None=Multiply(
    A=FunctionInput(Preview=@n19) {bUsePreviewValueAsDefault=true, inputName=Outline Glow Min, inputType=FunctionInput_Scalar, sortPriority=4}
    B=Custom() {code="return View.ViewportScaleUI;", description=ViewportScaleFactor, outputType=CMOT_Float1, inputs []}
   )) {name=Outline Glow Min, nodeColor=<rgba 0 1 0.0391 1>}
  n20 = Constant() {r=3}
  NamedRerouteDeclaration(None=Multiply(
    A=FunctionInput(Preview=@n20) {bUsePreviewValueAsDefault=true, inputName=Outline Glow Max, inputType=FunctionInput_Scalar, sortPriority=5}
    B=Custom() {code="return View.ViewportScaleUI;", description=ViewportScaleFactor, outputType=CMOT_Float1, inputs []}
   )) {name=Outline Glow Max, nodeColor=<rgba 0 1 0.0391 1>}
  n21 = Constant() {r=2}
  NamedRerouteDeclaration(None=Multiply(
    A=FunctionInput(Preview=@n21) {bUsePreviewValueAsDefault=true, inputName=Outline Thickness, inputType=FunctionInput_Scalar, sortPriority=1}
    B=Custom() {code="return View.ViewportScaleUI;", description=ViewportScaleFactor, outputType=CMOT_Float1, inputs []}
   )) {name=Outline Thickness, nodeColor=<rgba 0 1 0.0391 1>}
  NamedRerouteDeclaration(None=@n9) {name=Padding, nodeColor=<rgba 0 1 0 0>}
 ]
}
