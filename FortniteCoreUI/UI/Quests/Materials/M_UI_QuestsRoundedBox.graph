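# Translucent UI material that draws a rounded box shape, named for use in the quests UI.
# Four corner scalar parameters combine into a float4 roundness input for the MF_SDF_NormalizedRoundedBox function, giving each corner its own radius.
# Padding insets the shape and Falloff feeds the fill glow maximum.
# The function's shape fill output is multiplied by the Color parameter's alpha.
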Material /FortniteCoreUI/UI/Quests/Materials/M_UI_QuestsRoundedBox
graph {
 expressions {
  MaterialExpressionConvert_0 {
   class=MaterialExpressionConvert
   inputs [
    None {from=MaterialExpressionScalarParameter_1}
    None {from=MaterialExpressionScalarParameter_2}
    None {from=MaterialExpressionScalarParameter_3}
    None {from=MaterialExpressionScalarParameter_4}
   ]
   properties {
    nodeName=Make Float4
    convertInputs [
     {
      type=Scalar
      expressionInput {
       expression=&self:MaterialExpressionScalarParameter_1
       inputName=None
       mask=0
       maskA=0
       maskB=0
       maskG=0
       maskR=0
       outputIndex=0
      }
     }
     {
      type=Scalar
      expressionInput {
       expression=&self:MaterialExpressionScalarParameter_2
       inputName=None
       mask=0
       maskA=0
       maskB=0
       maskG=0
       maskR=0
       outputIndex=0
      }
     }
     {
      type=Scalar
      expressionInput {
       expression=&self:MaterialExpressionScalarParameter_3
       inputName=None
       mask=0
       maskA=0
       maskB=0
       maskG=0
       maskR=0
       outputIndex=0
      }
     }
     {
      type=Scalar
      expressionInput {
       expression=&self:MaterialExpressionScalarParameter_4
       inputName=None
       mask=0
       maskA=0
       maskB=0
       maskG=0
       maskR=0
       outputIndex=0
      }
     }
    ]
    convertOutputs [{type=Vector4, defaultValue=<rgba 0 0 0 1>}]
   }
  }
  MaterialExpressionMaterialFunctionCall_1 {
   class=MaterialExpressionMaterialFunctionCall
   inputs [
    NormalizedUVs {from=""}
    Outline Thickness {from=""}
    Padding {from=MaterialExpressionScalarParameter_6}
    Fill Glow Min {from=""}
    Fill Glow Max {from=MaterialExpressionScalarParameter_7}
    Outline Glow Min {from=""}
    Outline Glow Max {from=""}
    BoxCornerRoundness {from=MaterialExpressionConvert_0}
    UseHalfHeight {from=""}
   ]
   properties {materialFunction=/FortUILibrary/MaterialFunctions/Shapes/SDFs/MF_SDF_NormalizedRoundedBox}
  }
  MaterialExpressionMultiply_0 {
   class=MaterialExpressionMultiply
   inputs [
    A {from=MaterialExpressionMaterialFunctionCall_1, output=Shape Fill}
    B {from=MaterialExpressionVectorParameter_0, output=A}
   ]
  }
  MaterialExpressionScalarParameter_1 {class=MaterialExpressionScalarParameter, properties {defaultValue=16, parameterName=Corner1}}
  MaterialExpressionScalarParameter_2 {class=MaterialExpressionScalarParameter, properties {defaultValue=16, parameterName=Corner2}}
  MaterialExpressionScalarParameter_3 {class=MaterialExpressionScalarParameter, properties {defaultValue=16, parameterName=Corner3}}
  MaterialExpressionScalarParameter_4 {class=MaterialExpressionScalarParameter, properties {defaultValue=16, parameterName=Corner4}}
  MaterialExpressionScalarParameter_6 {class=MaterialExpressionScalarParameter, properties {defaultValue=4, parameterName=Padding}}
  MaterialExpressionScalarParameter_7 {class=MaterialExpressionScalarParameter, properties {parameterName=Falloff}}
  MaterialExpressionVectorParameter_0 {class=MaterialExpressionVectorParameter, properties {parameterName=Color}}
 }
 properties {
  bAutomaticallySetUsageInEditor=false
  bUsedWithStaticMesh=false
  blendMode=BLEND_Translucent
  materialDomain=MD_UI
  usedShadingModels=Default Lit
 }
}
