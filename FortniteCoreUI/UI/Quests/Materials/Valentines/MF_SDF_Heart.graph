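# UI material function that draws a heart shape from a signed distance field, despite descriptions that call it a regular polygon.
# Input UVs are centered, scaled by the HeartScale parameter, flipped in Y and mirrored in X.
# An If node picks between a circular-lobe distance and a signed minimum of two point/edge distances, the analytic heart SDF technique.
# MF_UI_SDFGlowOrAA then processes that distance into the SDF output.
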MaterialFunction /FortniteCoreUI/UI/Quests/Materials/Valentines/MF_SDF_Heart
graph {
 properties {
  bExposeToLibrary=true
  description=Signed distance field for a regular polygon with variable sides.
  previewMaterialDomain=MD_UI
  libraryCategoriesText [SDFs]
 }
 nodes [
  n1 = Add(A=NamedRerouteUsage(), B=NamedRerouteUsage())
  n2 = Constant() {r=1}
  n3 = Subtract(A=NamedRerouteUsage(), B=Constant2Vector() {g=0.75, r=0.25})
  n4 = Constant() {r=2}
  n5 = Subtract(A=SquareRoot(None=DotProduct(A=@n3, B=@n3)), B=Divide(A=SquareRoot(None=@n4), B=Constant() {r=4}))
  n6 = Subtract(A=NamedRerouteUsage(), B=Constant2Vector() {g=1})
  n7 = DotProduct(A=@n6, B=@n6)
  n8 = Add(A=NamedRerouteUsage(), B=NamedRerouteUsage())
  n9 = Constant()
  n10 = Subtract(A=NamedRerouteUsage(), B=Multiply(A=Constant() {r=0.5}, B=Max(A=@n8, B=@n9)))
  n11 = DotProduct(A=@n10, B=@n10)
  n12 = NamedRerouteUsage()
  n13 = NamedRerouteUsage()
  n14 = Multiply(A=SquareRoot(None=Min(A=@n7, B=@n11)), B=Sign(None=Subtract(A=@n12, B=@n13)))
  FunctionOutput(None=MaterialFunctionCall!(DistanceField=If(A=@n1, B=@n2, "A > B"=@n5, "A < B"=@n14), GlowMin, GlowMax) {materialFunction=/FortUILibrary/MaterialFunctions/Shapes/Modifiers/MF_UI_SDFGlowOrAA}.Result) {description=The resulting polygon SDF., outputName=SDF}
  n15 = FunctionInput(Preview=TextureCoordinate()) {
   bUsePreviewValueAsDefault=true
   description="Coordinates of the distance field (0,0 is origin)."
   inputName=UVs
   inputType=FunctionInput_Vector2
  }
  n16 = Constant4Vector() {constant=<rgba 0.5 0.28 1 0>}
  n17 = Divide(A=Subtract(A=@n15, B=AppendVector(A=@n16.R, B=@n16.R)), B=ScalarParameter() {defaultValue=0.5, parameterName=HeartScale})
  n18 = Abs(None=ComponentMask(None=@n17) {r=true})
  n19 = OneMinus(None=ComponentMask(None=@n17) {g=true})
  NamedRerouteDeclaration(None=AppendVector(
    A=NamedRerouteDeclaration(None=@n18) {name=HeartPx, nodeColor=<rgba 0 0.9579 1 1>}
    B=NamedRerouteDeclaration(None=@n19) {name=HeartPy, nodeColor=<rgba 0 0.3437 1 1>}
   )) {name=HeartP, nodeColor=<rgba 0.6641 1 0 1>}
 ]
}
