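# Valentine's UI material for the quest reward panel, drawing drifting pink heart particles.
# Two time-scrolled MF_UI_ParticleGrid layers feed an SDF heart function, with per-cell random values varying brightness, plus spin and scale in the first.
# Horizontal UV masks fade the edges and weight both layers rightward before they are summed and scaled by a radial gradient and Opacity.
# An animated oscillating-color blend serves as the exponent on the pink tint.
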Material /FortniteCoreUI/UI/Quests/Materials/M_UI_RewardPanel_ParticlesValentines
graph {
 properties {
  bAutomaticallySetUsageInEditor=false
  bUsedWithStaticMesh=false
  blendMode=BLEND_AlphaComposite
  materialDomain=MD_UI
  usedShadingModels=Default Lit
 }
 nodes [
  n1 = MaterialFunctionCall() {materialFunction=&GetUserInterfaceUV}
  n2 = Multiply(A=Time()) {constB=0.4173}
  n3 = LinearInterpolate(
   A=Constant3Vector() {constant=<rgba 1 1 1 1>}
   B=MaterialFunctionCall!(PrimaryUV, SecondaryUV, UVMix, PixelSize) {materialFunction=/FortUILibrary/MaterialFunctions/CoreUI/MF_UI_CoreUI_OscillatingColor}.Result
   Alpha=ComponentMask(None=Rotator(Coordinate=@n1.Normalized UV, Time=@n2) {centerX=1.5, speed=4.443}) {g=true}
  ) {constA=1}
  n4 = NamedRerouteUsage()
  n5 = Multiply(A=Round(None=@n4), B=@n4)
  n6 = MaterialFunctionCall(UV=MaterialFunctionCall() {materialFunction=&GetUserInterfaceUV}.Normalized UV) {materialFunction=&MF_UI_UVScaling}
  n7 = Time()
  n8 = Frac(None=Add(A=Multiply(A=@n7) {constB=0.1}, B=NamedRerouteUsage()))
  n9 = ScalarParameter() {parameterName=SmallTrianglesOffset}
  n10 = Time()
  n11 = NamedRerouteUsage()
  n12 = MaterialFunctionCall!(
   UV=@n6.Scaled to Fit
   Global Offset=AppendVector(A=Add(A=@n8, B=@n9), B=Constant())
   Global Rotation
   Tiling
   Cell Offset=Constant() {r=0.3}
   Cell Scale
   Cell Rotation=Multiply(A=Add(A=@n10, B=@n11)) {constB=0.03}
   Random Seed=ScalarParameter() {defaultValue=4.5, parameterName=Seed 1}
  ) {materialFunction=/FortUILibrary/MaterialFunctions/Basics/MF_UI_ParticleGrid}
  n13 = NamedRerouteDeclaration(None=@n12.Random Cell Values) {name=Random Cells, nodeColor=<rgba 0 0.8359 1 1>}
  n14 = Add(A=@n13, B=Time())
  n15 = MaterialFunctionCall!(UVs=MaterialFunctionCall(UVs=Rotator(Coordinate=@n12.Cells UV, Time=@n14) {centerX=0.4587, speed=0.058}, Texture Scale=LinearInterpolate(Alpha=@n13) {constA=0.3}) {materialFunction=&ScaleUVsByCenter}.UVs) {materialFunction=/FortniteCoreUI/UI/Quests/Materials/Valentines/MF_SDF_Heart}
  n16 = MaterialFunctionCall() {materialFunction=&GetUserInterfaceUV}
  n17 = ComponentMask(None=@n16.Normalized UV) {r=true}
  n18 = Multiply(A=Min(A=@n17, B=OneMinus(None=@n17))) {constB=2.5}
  n19 = Multiply(A=Multiply(A=Multiply(A=@n5, B=@n15.SDF)) {constB=0.5}, B=Multiply(A=@n18, B=@n18))
  n20 = ComponentMask(None=@n16.Normalized UV) {r=true}
  n21 = NamedRerouteUsage()
  n22 = Multiply(A=Round(None=@n21), B=@n21)
  n23 = Frac(None=Multiply(A=Time()) {constB=0.3})
  n24 = ScalarParameter() {parameterName=BigTrianglesOffset}
  n25 = Time()
  n26 = NamedRerouteUsage()
  n27 = MaterialFunctionCall(
   UV=MaterialFunctionCall(UVs=@n6.Scaled to Fit, Texture Scale=Constant2Vector() {g=1, r=1}) {materialFunction=&ScaleUVsByCenter}.UVs
   Global Offset=AppendVector(A=Add(A=@n23, B=@n24), B=Constant())
   Tiling=Constant() {r=3}
   Cell Offset=Constant() {r=0.05}
   Cell Rotation=Multiply(A=Add(A=@n25, B=@n26)) {constB=0.1}
   Random Seed=ScalarParameter() {defaultValue=5.2039, parameterName=Seed 2}
  ) {materialFunction=&MF_UI_ParticleGrid}
  n28 = MaterialFunctionCall(UVs=@n27.Cells UV) {materialFunction=&MF_SDF_Heart}
  n29 = Multiply(A=Power(Base=@n20) {constExponent=3}, B=Multiply(A=Multiply(A=@n22, B=@n28.SDF)) {constB=0.5})
  n30 = Saturate(None=Add(A=Multiply(A=@n19, B=@n20), B=Multiply(A=@n18, B=@n29)))
  n31 = ScalarParameter() {defaultValue=1, parameterName=Opacity, sliderMax=1}
  Multiply(
   A=Power(
    Base=Constant3Vector() {constant=<rgba 1 0.0122 0.124 1>}
    Exp=NamedRerouteDeclaration(None=@n3) {name=RiftOscillatingColors, nodeColor=<rgba 0 0 0 1>}
   )
   B=Multiply(
    A=MaterialFunctionCall() {materialFunction=&RadialGradientExponential}.RadialGradientExponential
    B=Multiply(A=@n30, B=@n31)
   )
  )
  NamedRerouteDeclaration(None=@n12.Random Rows) {name=Random Rows, nodeColor=<rgba 1 0.2812 0 1>}
  NamedRerouteDeclaration(None=@n27.Random Cell Values) {name=Random Cells_1, nodeColor=<rgba 0 0.8359 1 1>}
 ]
}
