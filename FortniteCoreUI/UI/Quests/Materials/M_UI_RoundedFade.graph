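# M_UI_RoundedFade is a translucent UI material in the Quests folder, built around a rounded-box SDF, apparently for a soft-edged fade.
# Texture coordinate channel 4, scaled by MultX and MultY parameters, feeds the normalized UVs of MF_SDF_NormalizedRoundedBox.
# A single PixelFalloff parameter drives both Padding and Fill Glow Max, tying the inset to the falloff width.
# A GetSlatePost1 call is present, and no links to material outputs appear.
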Material /FortniteCoreUI/UI/Quests/Materials/M_UI_RoundedFade
graph {
 properties {
  bUsedWithStaticMesh=false
  blendMode=BLEND_Translucent
  materialDomain=MD_UI
  pixelDepthOffsetMode=PDOM_Legacy
  usedShadingModels=Default Lit
 }
 nodes [
  MaterialFunctionCall!("UVs (Optional)") {materialFunction=/Engine/Functions/UserInterface/GetSlatePost1}
  n1 = ScalarParameter() {defaultValue=1, parameterName=MultX}
  n2 = ScalarParameter() {defaultValue=0.5, parameterName=MultY}
  n3 = ScalarParameter() {defaultValue=250, parameterName=PixelFalloff}
  MaterialFunctionCall!(
   NormalizedUVs=Multiply(A=TextureCoordinate() {coordinateIndex=4}, B=AppendVector(A=@n1, B=@n2))
   Outline Thickness
   Padding=@n3
   Fill Glow Min
   Fill Glow Max=@n3
   Outline Glow Min
   Outline Glow Max
   BoxCornerRoundness
   UseHalfHeight
  ) {materialFunction=/FortUILibrary/MaterialFunctions/Shapes/SDFs/MF_SDF_NormalizedRoundedBox}
 ]
}
