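# Test widget for previewing a quest category material across V3 and V2 quest UI layouts.
# A grid panel holds differently sized images for the V3, side nav, post-match, large tile, small tile and header variants.
# PreConstruct applies the TestMaterial variable to every image, then sets scalar switches on dynamic material instances (QuestV3, IsSideNav, IsLargeTile, IsHeader).
# The post-match image receives the material without parameter overrides.
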WidgetBlueprint /FortniteCoreUI/UI/Quests/Materials/WBP_QuestMaterialTestWidget
graphs [EventGraph {nodes [
   n1 = GetDynamicMaterial(execute=@n7.then, self=Get QuestV3() -> (QuestV3:Image Object Reference))
   n2 = SetScalarParameterValue(self=@n1, ParameterName=IsHeader, Value=0)
   SetScalarParameterValue(execute=[@n2.then, @n1.then], self=@n1, ParameterName=QuestV3, Value=1)
   > n3 = GetDynamicMaterial(self=Get SideNav() -> (SideNav:Image Object Reference))
   > SetScalarParameterValue(self=@n3, ParameterName=IsSideNav, Value=1)
   > n4 = GetDynamicMaterial(self=Get QuestV2_Large() -> (QuestV2_Large:Image Object Reference))
   > SetScalarParameterValue(self=@n4, ParameterName=IsLargeTile, Value=1)
   > n5 = GetDynamicMaterial(self=Get QuestV2_Small() -> (QuestV2_Small:Image Object Reference))
   > SetScalarParameterValue(self=@n5, ParameterName=IsLargeTile, Value=0)
   > n6 = GetDynamicMaterial(self=Get QuestV2_Header() -> (QuestV2_Header:Image Object Reference))
   > SetScalarParameterValue(self=@n6, ParameterName=IsHeader, Value=1)
   Event PreConstruct()
   > n7 = SetBrushFromMaterial(
    self=[
     Get QuestV3()
     Get QuestV2_Header()
     Get QuestV2_PostMatch() -> (QuestV2_PostMatch:Image Object Reference)
     Get SideNav()
     Get QuestV2_Large()
     Get QuestV2_Small()
    ]
    Material=Get TestMaterial() -> (TestMaterial:Material Interface Object Reference)
   )
  ]}]
variables [TestMaterial {type=MaterialInterface, default=/FortniteCoreUI/Items/QuestItems/Categories/Art/MI_UI_QB_RankedTrooper}]
widgets {
 info {inheritedWidgetCount=0, parentClass=&CommonUserWidget, rootWidgetClass=&GridPanel, widgetCount=9}
 widgets [
  GridPanel_15 {widgetClassPath=&GridPanel, widget=GridPanel_15}
  V3 {
   widgetClassPath=&CommonTextBlock
   parent=GridPanel_15
   widget=V3
   properties {
    text=V3
    font {
     bForceMonospaced=false
     bMaterialIsStencil=false
     fontMaterial=null
     fontObject=&HeadingNow
     letterSpacing=0
     monospacedWidth=1
     size=37.5
     skewAmount=0
     typefaceFontName=86 Bold
     outlineSettings {
      bApplyOutlineToDropShadows=false
      bMiteredCorners=false
      bSeparateFillAlpha=false
      outlineMaterial=null
      outlineSize=0
      outlineColor=<rgba 0 0 0 1>
     }
    }
    shadowOffset=<0 0>
   }
   slot {class=GridSlot}
  }
  QuestV3 {
   widgetClassPath=&Image
   bIsVariable=true
   parent=GridPanel_15
   widget=QuestV3
   properties {brush {
     bIsDynamicallyLoaded=false
     drawAs=Image
     imageType=FullColor
     mirroring=NoMirror
     resourceName=None
     resourceObject=null
     tiling=NoTile
     imageSize=<880 60>
     margin=<ltrb 0 0 0 0>
     outlineSettings {
      bUseBrushTransparency=false
      roundingType=HalfHeightRadius
      width=0
      color {colorUseRule=UseColor_Specified, specifiedColor=<rgba 0 0 0 0>}
      cornerRadii=<0 0 0 0>
     }
     tintColor {colorUseRule=UseColor_Specified, specifiedColor=<rgba 1 1 1 1>}
     uVRegion {bIsValid=false, max=<0 0>, min=<0 0>}
    }}
   slot {class=GridSlot, properties {horizontalAlignment=HAlign_Left, row=1}}
  }
  QuestV2_Large {
   widgetClassPath=&Image
   bIsVariable=true
   parent=GridPanel_15
   widget=QuestV2_Large
   properties {brush {
     bIsDynamicallyLoaded=false
     drawAs=Image
     imageType=FullColor
     mirroring=NoMirror
     resourceName=None
     resourceObject=null
     tiling=NoTile
     imageSize=<1112 140>
     margin=<ltrb 0 0 0 0>
     outlineSettings {
      bUseBrushTransparency=false
      roundingType=HalfHeightRadius
      width=0
      color {colorUseRule=UseColor_Specified, specifiedColor=<rgba 0 0 0 0>}
      cornerRadii=<0 0 0 0>
     }
     tintColor {colorUseRule=UseColor_Specified, specifiedColor=<rgba 1 1 1 1>}
     uVRegion {bIsValid=false, max=<0 0>, min=<0 0>}
    }}
   slot {class=GridSlot, properties {row=7, padding=<ltrb 0 40 0 0>}}
  }
  QuestV2_Small {
   widgetClassPath=&Image
   bIsVariable=true
   parent=GridPanel_15
   widget=QuestV2_Small
   properties {brush {
     bIsDynamicallyLoaded=false
     drawAs=Image
     imageType=FullColor
     mirroring=NoMirror
     resourceName=None
     resourceObject=null
     tiling=NoTile
     imageSize=<544 140>
     margin=<ltrb 0 0 0 0>
     outlineSettings {
      bUseBrushTransparency=false
      roundingType=HalfHeightRadius
      width=0
      color {colorUseRule=UseColor_Specified, specifiedColor=<rgba 0 0 0 0>}
      cornerRadii=<0 0 0 0>
     }
     tintColor {colorUseRule=UseColor_Specified, specifiedColor=<rgba 1 1 1 1>}
     uVRegion {bIsValid=false, max=<0 0>, min=<0 0>}
    }}
   slot {class=GridSlot, properties {horizontalAlignment=HAlign_Left, row=8, padding=<ltrb 0 40 0 0>}}
  }
  QuestV2_PostMatch {
   widgetClassPath=&Image
   bIsVariable=true
   parent=GridPanel_15
   widget=QuestV2_PostMatch
   properties {brush {
     bIsDynamicallyLoaded=false
     drawAs=Image
     imageType=FullColor
     mirroring=NoMirror
     resourceName=None
     resourceObject=/FortniteCoreUI/Items/QuestItems/Categories/Art/MI_UI_QB_S41_Weekly
     tiling=NoTile
     imageSize=<600 100>
     margin=<ltrb 0 0 0 0>
     outlineSettings {
      bUseBrushTransparency=false
      roundingType=HalfHeightRadius
      width=0
      color {colorUseRule=UseColor_Specified, specifiedColor=<rgba 0 0 0 0>}
      cornerRadii=<0 0 0 0>
     }
     tintColor {colorUseRule=UseColor_Specified, specifiedColor=<rgba 1 1 1 1>}
     uVRegion {bIsValid=false, max=<0 0>, min=<0 0>}
    }}
   slot {class=GridSlot, properties {horizontalAlignment=HAlign_Left, row=3, padding=<ltrb 0 40 0 0>}}
  }
  QuestV2_Header {
   widgetClassPath=&Image
   bIsVariable=true
   parent=GridPanel_15
   widget=QuestV2_Header
   properties {brush {
     bIsDynamicallyLoaded=false
     drawAs=Image
     imageType=FullColor
     mirroring=NoMirror
     resourceName=None
     resourceObject=null
     tiling=NoTile
     imageSize=<864 71>
     margin=<ltrb 0 0 0 0>
     outlineSettings {
      bUseBrushTransparency=false
      roundingType=HalfHeightRadius
      width=0
      color {colorUseRule=UseColor_Specified, specifiedColor=<rgba 0 0 0 0>}
      cornerRadii=<0 0 0 0>
     }
     tintColor {colorUseRule=UseColor_Specified, specifiedColor=<rgba 1 1 1 1>}
     uVRegion {bIsValid=false, max=<0 0>, min=<0 0>}
    }}
   slot {class=GridSlot, properties {horizontalAlignment=HAlign_Left, row=9, padding=<ltrb 0 40 0 0>}}
  }
  SideNav {
   widgetClassPath=&Image
   bIsVariable=true
   parent=GridPanel_15
   widget=SideNav
   properties {brush {
     bIsDynamicallyLoaded=false
     drawAs=Image
     imageType=FullColor
     mirroring=NoMirror
     resourceName=None
     resourceObject=null
     tiling=NoTile
     imageSize=<40 40>
     margin=<ltrb 0 0 0 0>
     outlineSettings {
      bUseBrushTransparency=false
      roundingType=HalfHeightRadius
      width=0
      color {colorUseRule=UseColor_Specified, specifiedColor=<rgba 0 0 0 0>}
      cornerRadii=<0 0 0 0>
     }
     tintColor {colorUseRule=UseColor_Specified, specifiedColor=<rgba 1 1 1 1>}
     uVRegion {bIsValid=false, max=<0 0>, min=<0 0>}
    }}
   slot {class=GridSlot, properties {horizontalAlignment=HAlign_Left, row=2, padding=<ltrb 0 40 0 0>}}
  }
  CommonTextBlock_75 {
   widgetClassPath=&CommonTextBlock
   parent=GridPanel_15
   widget=CommonTextBlock_75
   properties {
    text=V2
    font {
     bForceMonospaced=false
     bMaterialIsStencil=false
     fontMaterial=null
     fontObject=&HeadingNow
     letterSpacing=0
     monospacedWidth=1
     size=37.5
     skewAmount=0
     typefaceFontName=86 Bold
     outlineSettings {
      bApplyOutlineToDropShadows=false
      bMiteredCorners=false
      bSeparateFillAlpha=false
      outlineMaterial=null
      outlineSize=0
      outlineColor=<rgba 0 0 0 1>
     }
    }
    shadowOffset=<0 0>
   }
   slot {class=GridSlot, properties {row=6, padding=<ltrb 0 40 0 0>}}
  }
 ]
}
